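# Control Rig for the female medium IchorIncisor face-accessory part that adjusts hair bones to follow head, spine and shoulder poses.
# Head rotation, remapped to weights, offsets the hair master bones and three helper controls constrained to spine_05 and clavicles; back and side hair masters then aim at those controls.
# Long strands aim at spine_05.
# When WearingBackpack is true, mid-chain hair bones are constrained to upper-body bones, aimed at the pelvis and offset.
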ControlRigBlueprint /Game/Characters/Player/Female/Medium/Bodies/F_MED_IchorIncisor/Meshes/Parts/F_MED_IchorIncisor_FaceAcc_CtrlRig
graphs [
 RigVMFunctionLibrary {links [], nodes []}
 RigVMFunction_Sequence_1_ContainedGraph {nodes [
   n1 = RigVMFunctionEntryNode()
   RigVMFunction_Sequence = &RigVMFunction_Sequence(ExecuteContext=@n1.ExecuteContext)
   n2 = &RigVMFunction_Sequence(ExecuteContext=@RigVMFunction_Sequence.B)
   RigVMFunctionReturnNode(A=@RigVMFunction_Sequence.A, B=@n2.A, C=@n2.B)
  ]}
 RigVMModel {nodes [
   RigUnit_BeginExecution = &RigUnit_BeginExecution()
   RigVMFunction_Sequence = RigVMAggregateNode(ExecuteContext=@RigUnit_BeginExecution.ExecutePin)
   n1 = &RigUnit_GetRelativeTransformForItem(Child=(Type=Bone,Name="head"), Parent=(Type=Bone,Name="spine_04"), bChildInitial=False, bParentInitial=False)
   &RigUnit_ModifyTransforms(
    ItemToModify="((Item=(Type=Bone,Name=\"dyn_hair_bk_MSTR\"),Transform=(Rotation=(X=0,Y=0,Z=0,W=1),Translation=(X=-5,Y=-5,Z=0),Scale3D=(X=1,Y=1,Z=1))),(Item=(Type=Bone,Name=\"dyn_hair_l_MSTR\"),Transform=(Rotation=(X=0,Y=0,Z=0,W=1),Translation=(X=2.5,Y=-4,Z=0),Scale3D=(X=1,Y= ...(truncated)"
    WeightMinimum=0
    Weight=&RigVMFunction_MathDoubleRemap(
     SourceMaximum=-90
     SourceMinimum=0
     TargetMaximum=1
     TargetMinimum=0
     Value=&RigVMFunction_MathQuaternionToEuler(RotationOrder=ZXY, Value=@n1.RelativeTransform.Rotation).Result.Z
    ).Result
    ExecutePin=@RigVMFunction_Sequence.A
   )
   > &RigUnit_ParentConstraint(
    AdvancedSettings=(InterpolationType=Average,RotationOrderForFilter=XZY)
    Child=(Type=Control,Name="spine_05_ctrl")
    Parents=((Item=(Type=Bone,Name="spine_05"),Weight=1))
    Weight=1
   )
   > &RigUnit_ParentConstraint(
    AdvancedSettings=(InterpolationType=Average,RotationOrderForFilter=XZY)
    Child=(Type=Control,Name="spine_05_ctrl_2")
    Parents=((Item=(Type=Bone,Name="spine_05"),Weight=1),(Item=(Type=Bone,Name="clavicle_l"),Weight=1))
    Weight=1
   )
   > &RigUnit_ParentConstraint(
    AdvancedSettings=(InterpolationType=Average,RotationOrderForFilter=XZY)
    Child=(Type=Control,Name="spine_05_ctrl_2_2")
    Parents=((Item=(Type=Bone,Name="spine_05"),Weight=1),(Item=(Type=Bone,Name="clavicle_r"),Weight=1))
    Weight=1
   )
   n2 = &RigUnit_GetRelativeTransformForItem(Child=(Type=Bone,Name="head"), Parent=(Type=Bone,Name="neck_02"), bChildInitial=False, bParentInitial=False)
   > &RigUnit_ModifyTransforms(
    ItemToModify="((Item=(Type=Control,Name=\"spine_05_ctrl\"),Transform=(Rotation=(X=0,Y=0,Z=0,W=1),Translation=(X=1,Y=0,Z=4),Scale3D=(X=1,Y=1,Z=1))),(Item=(Type=Control,Name=\"spine_05_ctrl_2\"),Transform=(Rotation=(X=0,Y=0,Z=0,W=1),Translation=(X=0,Y=0,Z=2),Scale3D=(X=1,Y=1, ...(truncated)"
    WeightMinimum=0
    Weight=&RigVMFunction_MathDoubleRemap(
     SourceMaximum=-30
     SourceMinimum=0
     TargetMaximum=1
     TargetMinimum=0
     Value=&RigVMFunction_MathQuaternionToEuler(RotationOrder=ZXY, Value=@n2.RelativeTransform.Rotation).Result.X
    ).Result
   )
   n3 = &RigUnit_GetRelativeTransformForItem(Child=(Type=Bone,Name="head"), Parent=(Type=Bone,Name="neck_02"), bChildInitial=False, bParentInitial=False)
   > &RigUnit_ModifyTransforms(
    ItemToModify="((Item=(Type=Control,Name=\"spine_05_ctrl\"),Transform=(Rotation=(X=0,Y=0,Z=0,W=1),Translation=(X=1,Y=0,Z=-4),Scale3D=(X=1,Y=1,Z=1))),(Item=(Type=Control,Name=\"spine_05_ctrl_2\"),Transform=(Rotation=(X=0,Y=0,Z=0,W=1),Translation=(X=0,Y=0,Z=-6),Scale3D=(X=1,Y= ...(truncated)"
    WeightMinimum=0
    Weight=&RigVMFunction_MathDoubleRemap(
     SourceMaximum=30
     SourceMinimum=0
     TargetMaximum=1
     TargetMinimum=0
     Value=&RigVMFunction_MathQuaternionToEuler(RotationOrder=ZXY, Value=@n3.RelativeTransform.Rotation).Result.X
    ).Result
   )
   > &RigUnit_AimConstraintLocalSpaceOffset(
    Child=(Type=Bone,Name="dyn_hair_bk_MSTR")
    Filter=(bX=True,bY=True,bZ=True)
    Parents=((Item=(Type=Control,Name="spine_05_ctrl"),Weight=1))
    Weight=1
    WorldUp=(Target=(X=0,Y=0,Z=1),Kind=Location,Space=(Type=Bone,Name="spine_05"))
    bMaintainOffset=true
   )
   > &RigUnit_AimConstraintLocalSpaceOffset(
    Child=(Type=Bone,Name="dyn_hair_l_MSTR")
    Filter=(bX=True,bY=True,bZ=True)
    Parents=((Item=(Type=Control,Name="spine_05_ctrl_2_2"),Weight=1))
    Weight=1
    WorldUp=(Target=(X=0,Y=0,Z=1),Kind=Location,Space=(Type=Bone,Name="spine_05"))
    bMaintainOffset=true
   )
   > &RigUnit_AimConstraintLocalSpaceOffset(
    Child=(Type=Bone,Name="dyn_hair_r_MSTR")
    Filter=(bX=True,bY=True,bZ=True)
    Parents=((Item=(Type=Control,Name="spine_05_ctrl_2"),Weight=1))
    Weight=1
    WorldUp=(Target=(X=0,Y=0,Z=1),Kind=Location,Space=(Type=Bone,Name="spine_05"))
    bMaintainOffset=true
   )
   &RigUnit_AimConstraintLocalSpaceOffset(
    Child=(Type=Bone,Name="dyn_stland_long_l_MSTR")
    Filter=(bX=True,bY=True,bZ=True)
    Parents=((Item=(Type=Bone,Name="spine_05"),Weight=1))
    Weight=1
    WorldUp=(Target=(X=0,Y=0,Z=1),Kind=Location,Space=(Type=Bone,Name="spine_05"))
    ExecutePin=@RigVMFunction_Sequence.B
   )
   > &RigUnit_AimConstraintLocalSpaceOffset(
    Child=(Type=Bone,Name="dyn_strand_long_r_MSTR")
    Filter=(bX=True,bY=True,bZ=True)
    Parents=((Item=(Type=Bone,Name="spine_05"),Weight=1))
    Weight=1
    WorldUp=(Target=(X=0,Y=0,Z=1),Kind=Location,Space=(Type=Bone,Name="spine_05"))
   )
   RigVMFunction_ControlFlowBranch = &RigVMFunction_ControlFlowBranch(ExecuteContext=@RigVMFunction_Sequence.C, Condition=RigVMVariableNode() {variable=WearingBackpack}.Value)
   &RigUnit_ParentConstraint(
    AdvancedSettings=(RotationOrderForFilter=XZY)
    Child=(Type=Bone,Name="dyn_hair_l_03")
    Parents=((Item=(Type=Bone,Name="upperarm_r"),Weight=0.0198),(Item=(Type=Bone,Name="spine_05"),Weight=0.3295),(Item=(Type=Bone,Name="clavicle_r"),Weight=0.6028))
    Weight=1
    ExecutePin=@RigVMFunction_ControlFlowBranch.True
   )
   > &RigUnit_ParentConstraint(
    AdvancedSettings=(RotationOrderForFilter=XZY)
    Child=(Type=Bone,Name="dyn_hair_bk_03")
    Parents=((Item=(Type=Bone,Name="clavicle_r"),Weight=0.1777),(Item=(Type=Bone,Name="clavicle_l"),Weight=0.1777),(Item=(Type=Bone,Name="spine_05"),Weight=0.6447))
    Weight=1
   )
   > &RigUnit_ParentConstraint(
    AdvancedSettings=(RotationOrderForFilter=XZY)
    Child=(Type=Bone,Name="dyn_hair_r_03")
    Parents=((Item=(Type=Bone,Name="upperarm_l"),Weight=0.0378),(Item=(Type=Bone,Name="spine_05"),Weight=0.1779),(Item=(Type=Bone,Name="clavicle_l"),Weight=0.6551))
    Weight=1
   )
   > &RigUnit_AimConstraintLocalSpaceOffset(
    Child=(Type=Bone,Name="dyn_hair_bk_03")
    Filter=(bX=True,bY=True,bZ=True)
    Parents=((Item=(Type=Bone,Name="pelvis"),Weight=1))
    Weight=1
    WorldUp=(Target=(X=0,Y=0,Z=1),Kind=Location,Space=(Type=Bone,Name="pelvis"))
   )
   > &RigUnit_AimConstraintLocalSpaceOffset(
    Child=(Type=Bone,Name="dyn_hair_l_03")
    Filter=(bX=True,bY=True,bZ=True)
    Parents=((Item=(Type=Bone,Name="pelvis"),Weight=1))
    Weight=1
    WorldUp=(Target=(X=0,Y=0,Z=1),Kind=Location,Space=(Type=Bone,Name="pelvis"))
   )
   > &RigUnit_AimConstraintLocalSpaceOffset(
    Child=(Type=Bone,Name="dyn_hair_r_03")
    Filter=(bX=True,bY=True,bZ=True)
    Parents=((Item=(Type=Bone,Name="pelvis"),Weight=1))
    Weight=1
    WorldUp=(Target=(X=0,Y=0,Z=1),Kind=Location,Space=(Type=Bone,Name="pelvis"))
   )
   > &RigUnit_ModifyTransforms(
    ItemToModify="((Item=(Type=Bone,Name=\"dyn_hair_bk_03\"),Transform=(Rotation=(X=0,Y=0,Z=0.0872,W=0.9962),Translation=(X=0,Y=-2,Z=0),Scale3D=(X=1,Y=1,Z=1))),(Item=(Type=Bone,Name=\"dyn_hair_l_03\"),Transform=(Rotation=(X=0.0038,Y=-0.0435,Z=0.0871,W=0.9952),Translation=(X=0,Y ...(truncated)"
    Weight=1
    WeightMinimum=0
   )
  ]}
]
hierarchy [
 root {loc=<0 0 0>, rot=<0 0 0 1>}
 pelvis {parent=root, loc=<0 -0.5301 98.0571>, rot=<0.0127 0.707 -0.0127 -0.707>}
 spine_01 {parent=pelvis, loc=<6.1208 -0.9328 0>, rot=<0 0 0.034 -0.9994>}
 spine_02 {parent=spine_01, loc=<6.5996 -0.2095 0>, rot=<0 0 0.0238 -0.9997>}
 spine_03 {parent=spine_02, loc=<6.4914 -0.1237 0>, rot=<0 0 0.0149 -0.9999>}
 spine_04 {parent=spine_03, loc=<6.9598 0.0794 0>, rot=<0 0 -0.0132 -0.9999>}
 spine_05 {parent=spine_04, loc=<6.9987 0 0>, rot=<0 0 -0.0015 -1>}
 neck_01 {parent=spine_05, loc=<11.3138 0.1529 0>, rot=<0 0 -0.2003 -0.9797>}
 neck_02 {parent=neck_01, loc=<4.5627 0 0>, rot=<0 0 0.0504 -0.9987>}
 head {parent=neck_02, loc=<5.3914 0 0>, rot=<0 0 0.115 -0.9934>}
 glasses {parent=head, loc=<2.0583 9.491 -0.1246>, rot=<0.0922 -0.7011 -0.7011 0.0922>}
 dyn_strand_short_l_MSTR {parent=head, loc=<7.4958 1.6475 -6.3431>, rot=<-0.0813 -0.9923 -0.0084 -0.0933>}
 dyn_strand_short_l_01 {parent=dyn_strand_short_l_MSTR, loc=<0 0 0>, rot=<0 0 0 1>}
 dyn_strand_short_l_02 {parent=dyn_strand_short_l_01, loc=<5.8096 0 0>, rot=<-0.0046 -0.0573 0.0916 -0.9941>}
 dyn_strand_short_l_03 {parent=dyn_strand_short_l_02, loc=<4.5887 0 0>, rot=<0 0 0 1>}
 dyn_piercing {parent=head, loc=<-2.6884 8.219 -0.1249>, rot=<0.0922 -0.7011 -0.7011 0.0922>}
 dyn_stland_long_l_MSTR {parent=head, loc=<-0.0548 -2.0799 -5.7127>, rot=<-0.0735 -0.9945 -0.006 -0.0743>}
 dyn_stland_long_l_01 {parent=dyn_stland_long_l_MSTR, loc=<0 0 0>, rot=<0 0 -0.0001 -1>}
 dyn_stland_long_l_02 {parent=dyn_stland_long_l_01, loc=<5.3477 0 0>, rot=<0.0406 0.1167 -0.1775 -0.9763>}
 dyn_stland_long_l_03 {parent=dyn_stland_long_l_02, loc=<5.2689 0.0005 -0.0002>, rot=<0.0001 0 0 -1>}
 dyn_stland_short_r_MSTR {parent=head, loc=<7.4954 1.6475 6.3431>, rot=<-0.0813 -0.9923 0.0084 0.0933>}
 dyn_stland_short_r_01 {parent=dyn_stland_short_r_MSTR, loc=<0 0 0>, rot=<0 0 0 1>}
 dyn_stland_short_r_02 {parent=dyn_stland_short_r_01, loc=<5.8088 0.0001 -0.0002>, rot=<0.0046 0.0573 0.0916 -0.9941>}
 dyn_stland_short_r_03 {parent=dyn_stland_short_r_02, loc=<4.5894 0 0>, rot=<0 0 0 1>}
 dyn_strand_long_r_MSTR {parent=head, loc=<-0.0517 -2.08 5.7131>, rot=<-0.0732 -0.9945 0.0061 0.0742>}
 dyn_strand_long_r_01 {parent=dyn_strand_long_r_MSTR, loc=<0 0 0>, rot=<0 0 0 1>}
 dyn_strand_long_r_02 {parent=dyn_strand_long_r_01, loc=<5.349 0 0>, rot=<-0.0406 -0.1167 -0.1775 -0.9763>}
 dyn_strand_long_r_03 {parent=dyn_strand_long_r_02, loc=<5.2681 0 0>, rot=<0 0 0 1>}
 dyn_hair_bk_MSTR {parent=head, loc=<4.0186 -6.4113 -0.144>, rot=<0.1517 -0.9884 0 0>}
 dyn_hair_bk_01 {parent=dyn_hair_bk_MSTR, loc=<0 0 0>, rot=<0 0 0 1>}
 dyn_hair_bk_02 {parent=dyn_hair_bk_01, loc=<10.0998 0 0>, rot=<0 0 0.0288 -0.9996>}
 dyn_hair_bk_03 {parent=dyn_hair_bk_02, loc=<8.688 0 0>, rot=<0 0 -0.0321 -0.9995>}
 dyn_hair_bk_04 {parent=dyn_hair_bk_03, loc=<12.4243 -0.0518 0>, rot=<0 0 0.0059 -1>}
 dyn_hair_bk_05 {parent=dyn_hair_bk_04, loc=<10.802 0 0>, rot=<0 0 -0.016 -0.9999>}
 dyn_hair_l_MSTR {parent=head, loc=<3.7002 -4.6513 4.666>, rot=<0.1424 -0.9825 0.119 0.0171>}
 dyn_hair_l_01 {parent=dyn_hair_l_MSTR, loc=<0 0 0>, rot=<0 0 0 1>}
 dyn_hair_l_02 {parent=dyn_hair_l_01, loc=<9.769 0 0>, rot=<0.0059 -0.0156 0.0947 -0.9954>}
 dyn_hair_l_03 {parent=dyn_hair_l_02, loc=<8.5676 0 0>, rot=<0.0144 -0.0359 -0.0733 -0.9966>}
 dyn_hair_l_05 {parent=dyn_hair_l_03, loc=<12.6322 0.0384 -0.0208>, rot=<0.0033 -0.0111 -0.0205 -0.9997>}
 dyn_hair_l_06 {parent=dyn_hair_l_05, loc=<9.6369 0 0>, rot=<0 0 0 1>}
 dyn_hair_r_MSTR {parent=head, loc=<3.7671 -4.6508 -5.6373>, rot=<0.1502 -0.9773 -0.1485 -0.0141>}
 dyn_hair_r_01 {parent=dyn_hair_r_MSTR, loc=<0 0 0>, rot=<0 0 0 1>}
 dyn_hair_r_02 {parent=dyn_hair_r_01, loc=<9.7697 0.0002 0>, rot=<-0.0059 0.0156 0.0947 -0.9954>}
 dyn_hair_r_03 {parent=dyn_hair_r_02, loc=<8.5672 -0.0003 0.0001>, rot=<-0.0144 0.0359 -0.0733 -0.9966>}
 dyn_hair_r_05 {parent=dyn_hair_r_03, loc=<11.9169 0.0092 0.0048>, rot=<-0.0033 0.0111 -0.0205 -0.9997>}
 dyn_hair_r_06 {parent=dyn_hair_r_05, loc=<11.2714 0 0>, rot=<0 0 0 1>}
 dyn_fringe_r_MSTR {parent=head, loc=<9.4913 8.183 3.7935>, rot=<-0.1278 -0.9119 -0.389 0.0284>}
 dyn_fringe_r_01 {parent=dyn_fringe_r_MSTR, loc=<0 0 0>, rot=<0 0 0 1>}
 dyn_fringe_r_02 {parent=dyn_fringe_r_01, loc=<4.2645 0 0>, rot=<0 0 0 1>}
 dyn_fringe_fr_MSTR {parent=head, loc=<9.4908 9.7912 -1.1334>, rot=<-0.1132 -0.9935 -0.001 -0.0084>}
 dyn_fringe_fr_01 {parent=dyn_fringe_fr_MSTR, loc=<0 0 0>, rot=<0 0 0 1>}
 dyn_fringe_fr_02 {parent=dyn_fringe_fr_01, loc=<4.7695 0 0>, rot=<0 0 0 1>}
 dyn_fringe_l_MSTR {parent=head, loc=<9.2114 7.0455 -5.548>, rot=<-0.1014 -0.8802 0.4632 -0.0193>}
 dyn_fringe_l_01 {parent=dyn_fringe_l_MSTR, loc=<0 0 0>, rot=<0 0 0 1>}
 dyn_fringe_l_02 {parent=dyn_fringe_l_01, loc=<4.0758 0 0>, rot=<0 0 0 1>}
 FACIAL_C_FacialRoot {parent=head, loc=<1.7377 0.1739 0>, rot=<0.5021 -0.4979 -0.5021 -0.4979>}
 FACIAL_C_LowerLipRotation {parent=FACIAL_C_FacialRoot, loc=<0 2.0307 1.3395>, rot=<0.342 0 0 -0.9397>}
 FACIAL_C_MouthLower {parent=FACIAL_C_LowerLipRotation, loc=<0 -1.6943 5.3537>, rot=<-0.342 0 0 -0.9397>}
 FACIAL_C_LipLower {parent=FACIAL_C_MouthLower, loc=<0 0.7216 1.722>, rot=<0.0872 0 0 -0.9962>}
 FACIAL_C_Jaw {parent=FACIAL_C_FacialRoot, loc=<0 2.0307 1.3395>, rot=<0.342 0 0 -0.9397>}
 clavicle_l {parent=spine_05, loc=<5.4831 -0.1053 -3.2946>, rot=<0.0465 -0.7289 -0.009 -0.6829>}
 upperarm_l {parent=clavicle_l, loc=<9.8658 -1.2136 -0.4105>, rot=<0.0164 -0.3245 -0.0269 -0.9454>}
 clavicle_r {parent=spine_05, loc=<5.4831 -0.1053 3.2946>, rot=<0.7289 0.0465 -0.6829 0.009>}
 upperarm_r {parent=clavicle_r, loc=<-9.8658 1.2136 0.4105>, rot=<0.0164 -0.3245 -0.0269 -0.9454>}
 spine_05_ctrl {controlType=EULER_TRANSFORM, parent=spine_05, type=Control, loc=<0 0 0>, rot=<0 0 0 1>}
 spine_05_ctrl_2 {controlType=EULER_TRANSFORM, parent=spine_05, type=Control, loc=<0 0 0>, rot=<0 0 0 1>}
 spine_05_ctrl_2_2 {controlType=EULER_TRANSFORM, parent=spine_05, type=Control, loc=<0 0 0>, rot=<0 0 0 1>}
]
variables [WearingBackpack {type=bool, default=False}]
