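# Physics asset for the IchorIncisor medium female player's face accessory (WithBack variant), defining capsule bodies for dynamic hair.
# Head, neck, spine, pelvis, clavicle and upper-arm capsules cover the body around the hair.
# Back, left and right hair form two-joint chains from their 03 bones; two long strands form three-joint chains from MSTR roots.
# Constraints set tight swing and twist limits; hair MSTR bodies are unconstrained, and left-strand bones are spelled "stland".
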PhysicsAsset /Game/Characters/Player/Female/Medium/Bodies/F_MED_IchorIncisor/Meshes/Parts/F_MED_IchorIncisor_FaceAcc_Physics_WithBack
bodies {
 clavicle_l [{shapeType=Capsule, length=6.5, radius=6, shapeName=clavicle_l_capsule, center=<4.2319 -1.2308 0.5824>, rotation=<rot 75.0961 1.7248 -0.6378>}]
 clavicle_r [{
   shapeType=Capsule
   length=4.5
   radius=6
   shapeName=clavicle_r_capsule
   center=<-6.4195 1.1387 -0.007>
   rotation=<rot 84.9995 -0.1009 -0.8615>
  }]
 dyn_hair_bk_03 [{shapeType=Capsule, length=2, radius=0.5, shapeName=dyn_hair_bk_03_capsule}]
 dyn_hair_bk_04 [{shapeType=Capsule, length=7.5, radius=6.5, shapeName=dyn_hair_bk_04_capsule, center=<8 -5 0>}]
 dyn_hair_bk_05 [{shapeType=Capsule, length=15.9572, radius=6.3298, shapeName=dyn_hair_bk_05_capsule, center=<9.4 -5.2678 0>}]
 dyn_hair_bk_MSTR [{shapeType=Capsule, length=2, radius=0.5, shapeName=dyn_hair_bk_MSTR_capsule}]
 dyn_hair_l_03 [{shapeType=Capsule, length=2, radius=0.5, shapeName=dyn_hair_l_03_capsule}]
 dyn_hair_l_05 [{shapeType=Capsule, length=7.5, radius=2.5, shapeName=dyn_hair_l_05_capsule, center=<5 -0.8 -2.6>, rotation=<rot 88.9316 2.7323 3.6131>}]
 dyn_hair_l_06 [{
   shapeType=Capsule
   length=17.5095
   radius=4
   shapeName=dyn_hair_l_06_capsule
   center=<5.2873 -1.84 -1.3907>
   rotation=<rot 86.8946 0.0013 3.1463>
  }]
 dyn_hair_l_MSTR [{shapeType=Capsule, length=2, radius=0.5, shapeName=dyn_hair_l_MSTR_capsule}]
 dyn_hair_r_03 [{shapeType=Capsule, length=2, radius=0.5, shapeName=dyn_hair_r_03_capsule}]
 dyn_hair_r_05 [{shapeType=Capsule, length=7.5, radius=2.5, shapeName=dyn_hair_r_05_capsule, center=<6 0.68 0.87>, rotation=<rot 86.4264 -176.4795 -177.0162>}]
 dyn_hair_r_06 [{shapeType=Capsule, length=13.6008, radius=4.2528, shapeName=dyn_hair_r_06_capsule, center=<5.3495 -1 2.5>, rotation=<rot 90 180 -174.9608>}]
 dyn_hair_r_MSTR [{shapeType=Capsule, length=2, radius=0.5, shapeName=dyn_hair_r_MSTR_capsule}]
 dyn_stland_long_l_01 [{shapeType=Capsule, length=2.5, radius=3, shapeName=dyn_stland_long_l_01_capsule, center=<2.8 0.9 0>, rotation=<rot -90 180 174.4414>}]
 dyn_stland_long_l_02 [{
   shapeType=Capsule
   length=5
   radius=3
   shapeName=dyn_stland_long_l_02_capsule
   center=<3.4436 1.3418 1.1523>
   rotation=<rot -90 180 -173.0395>
  }]
 dyn_stland_long_l_03 [{
   shapeType=Capsule
   length=5.5
   radius=3
   shapeName=dyn_stland_long_l_03_capsule
   center=<2.4518 3.5925 -1.3136>
   rotation=<rot -45.4685 -172.8027 -158.9129>
  }]
 dyn_stland_long_l_MSTR [{shapeType=Capsule, length=1, radius=0.5, shapeName=dyn_stland_long_l_MSTR_capsule}]
 dyn_strand_long_r_01 [{shapeType=Capsule, length=2.5, radius=3, shapeName=dyn_strand_long_r_01_capsule, center=<2.54 0 -1.63>, rotation=<rot -90 -180 180>}]
 dyn_strand_long_r_02 [{
   shapeType=Capsule
   length=5.5
   radius=3
   shapeName=dyn_strand_long_r_02_capsule
   center=<3.3 -0.22 -2.44>
   rotation=<rot -83.7914 -179.972 179.9783>
  }]
 dyn_strand_long_r_03 [{shapeType=Capsule, length=6, radius=3, shapeName=dyn_strand_long_r_03_capsule, center=<3.3 2.2 -0.7>, rotation=<rot -61 30 0>}]
 dyn_strand_long_r_MSTR [{shapeType=Capsule, length=1, radius=0.5, shapeName=dyn_strand_long_r_MSTR_capsule}]
 head [{shapeType=Capsule, length=6.1, radius=6.5, shapeName=head_capsule, center=<2.6 2.5 0>, rotation=<rot 90 0 0>}]
 neck_01 [{shapeType=Capsule, length=4.7, radius=5.5, shapeName=neck_01_capsule, center=<2.6 0.6447 0>, rotation=<rot 90 0 0>}]
 pelvis [{shapeType=Capsule, length=6.5, radius=13.5, shapeName=pelvis_capsule, center=<-4 1 0>}]
 spine_02 [{shapeType=Capsule, length=2.2884, radius=14.1977, shapeName=spine_02_capsule, center=<0 3.8247 0>}]
 spine_04 [{shapeType=Capsule, length=3.5, radius=13, shapeName=spine_04_capsule, center=<-3.3017 4.0831 0>}]
 spine_05 [{shapeType=Capsule, length=5.75, radius=10.6, shapeName=spine_05_capsule, center=<-0.8495 1.765 0>}]
 upperarm_l [{shapeType=Capsule, length=27.5, radius=4.8, shapeName=upperarm_l_capsule, center=<14 -0.5 0>, rotation=<rot 90 0 0>}]
 upperarm_r [{shapeType=Capsule, length=27.5, radius=4.8, shapeName=upperarm_r_capsule, center=<-14 0.5 0>, rotation=<rot 90 0 0>}]
}
constraints [
 {bone1Name=dyn_hair_bk_05, bone2Name=dyn_hair_bk_04, swing1LimitDegrees=20, swing2LimitDegrees=15, twistLimitDegrees=2}
 {bone1Name=dyn_hair_l_06, bone2Name=dyn_hair_l_05, swing1LimitDegrees=20, swing2LimitDegrees=15, twistLimitDegrees=2}
 {bone1Name=dyn_hair_r_06, bone2Name=dyn_hair_r_05, swing1LimitDegrees=20, swing2LimitDegrees=15, twistLimitDegrees=2}
 {bone1Name=dyn_stland_long_l_01, bone2Name=dyn_stland_long_l_MSTR, swing1LimitDegrees=20, swing2LimitDegrees=10, twistLimitDegrees=10}
 {bone1Name=dyn_stland_long_l_02, bone2Name=dyn_stland_long_l_01, swing1LimitDegrees=20, swing2LimitDegrees=10, twistLimitDegrees=10}
 {bone1Name=dyn_stland_long_l_03, bone2Name=dyn_stland_long_l_02, swing1LimitDegrees=10, swing2LimitDegrees=10, twistLimitDegrees=10}
 {bone1Name=dyn_strand_long_r_01, bone2Name=dyn_strand_long_r_MSTR, swing1LimitDegrees=20, swing2LimitDegrees=10, twistLimitDegrees=10}
 {bone1Name=dyn_strand_long_r_02, bone2Name=dyn_strand_long_r_01, swing1LimitDegrees=20, swing2LimitDegrees=10, twistLimitDegrees=10}
 {bone1Name=dyn_strand_long_r_03, bone2Name=dyn_strand_long_r_02, swing1LimitDegrees=10, swing2LimitDegrees=10, twistLimitDegrees=10}
 {bone1Name=dyn_hair_l_05, bone2Name=dyn_hair_l_03, swing1LimitDegrees=30, swing2LimitDegrees=5, twistLimitDegrees=2}
 {bone1Name=dyn_hair_bk_04, bone2Name=dyn_hair_bk_03, swing1LimitDegrees=30, swing2LimitDegrees=5, twistLimitDegrees=2}
 {bone1Name=dyn_hair_r_05, bone2Name=dyn_hair_r_03, swing1LimitDegrees=30, swing2LimitDegrees=5, twistLimitDegrees=2}
]
