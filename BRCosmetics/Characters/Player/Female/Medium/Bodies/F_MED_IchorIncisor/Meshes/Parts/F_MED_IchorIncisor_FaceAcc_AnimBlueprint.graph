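# Post-process style AnimBlueprint for the IchorIncisor female face accessory, animating hair, fringe, strands and lip piercing.
# It copies the parent mesh pose through a Control Rig, then applies bone constraints and spring AnimDynamics to form PreSimPose.
# The output blends two RigidBody simulations of that pose, choosing physics assets by Wearing_Backpack.
# More elaborate skydive, parachute and jump hair branches exist but do not reach the output.
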AnimBlueprint /BRCosmetics/Characters/Player/Female/Medium/Bodies/F_MED_IchorIncisor/Meshes/Parts/F_MED_IchorIncisor_FaceAcc_AnimBlueprint
target_skeleton=/Game/Characters/Player/Female/Medium/Bodies/F_MED_IchorIncisor/Meshes/Parts/F_MED_IchorIncisor_FaceAcc_Skeleton
graphs [
 AnimGraph {nodes [
   n1 = "Anim Dynamics - Chain: dyn_hair_r_05 -> dyn_hair_r_06"!(ComponentPose, Alpha=1, ComponentLinearAccScale:Vector="0.5,0.5,0.5", ComponentLinearVelScale:Vector="0.5,0.5,0.5", ExternalForce="0,0,0") -> (Pose) {
    angularDampingOverride=0.999
    angularSpringConstant=200
    bAngularSpring=true
    bChain=true
    bOverrideAngularDamping=true
    bUseGravityOverride=true
    bUseSphericalLimits=true
    simulationSpace=RootRelative
    alphaBoolBlend {blendOption=Linear}
    boundBone {boneName=dyn_hair_r_05}
    chainEnd {boneName=dyn_hair_r_06}
    componentAppliedLinearAccClamp=<2000 2000 2000>
    componentLinearAccScale=<0.5 0.5 0.5>
    componentLinearVelScale=<0.5 0.5 0.5>
    gravityOverride {z=-980}
    physicsBodyDefinitions [
     {
      boundBone {boneName=dyn_hair_r_05}
      boxExtents=<30 30 30>
      constraintSetup {angularLimitsMax=<0 10 20>, angularLimitsMin=<0 -10 -20>, angularTarget=<0 0 0>}
      localJointOffset=<1 0 0>
     }
     {
      boundBone {boneName=dyn_hair_r_06}
      boxExtents=<30 30 30>
      constraintSetup {angularLimitsMax=<0 30 30>, angularLimitsMin=<0 -30 -30>, angularTarget=<0 0 0>}
      localJointOffset=<1 0 0>
     }
    ]
    relativeSpaceBone {boneName=spine_05}
    sphericalLimits [
     {limitRadius=13, drivingBone {boneName=pelvis}, sphereLocalOffset=<2 -5 6>}
     {limitRadius=11, drivingBone {boneName=spine_02}, sphereLocalOffset=<0 0 4>}
     {limitRadius=13, drivingBone {boneName=pelvis}, sphereLocalOffset=<2 -5 -6>}
     {limitRadius=11, drivingBone {boneName=spine_02}, sphereLocalOffset=<0 0 -4>}
    ]
   }
   n2 = "Anim Dynamics - Chain: dyn_hair_l_05 -> dyn_hair_l_06"!(
    ComponentPose=@n1
    Alpha=1
    ComponentLinearAccScale:Vector="0.5,0.5,0.5"
    ComponentLinearVelScale:Vector="0.5,0.5,0.5"
    ExternalForce="0,0,0"
   ) -> (Pose) {
    angularDampingOverride=0.999
    angularSpringConstant=200
    bAngularSpring=true
    bChain=true
    bOverrideAngularDamping=true
    bUseGravityOverride=true
    bUseSphericalLimits=true
    alphaBoolBlend {blendOption=Linear}
    boundBone {boneName=dyn_hair_l_05}
    chainEnd {boneName=dyn_hair_l_06}
    componentAppliedLinearAccClamp=<2000 2000 2000>
    componentLinearAccScale=<0.5 0.5 0.5>
    componentLinearVelScale=<0.5 0.5 0.5>
    gravityOverride {z=-980}
    physicsBodyDefinitions [
     {
      boundBone {boneName=dyn_hair_l_05}
      boxExtents=<30 30 30>
      constraintSetup {angularLimitsMax=<0 10 20>, angularLimitsMin=<0 -10 -20>, angularTarget=<0 0 0>}
      localJointOffset=<-1 0 0>
     }
     {
      boundBone {boneName=dyn_hair_l_06}
      boxExtents=<30 30 30>
      constraintSetup {angularLimitsMax=<0 30 30>, angularLimitsMin=<0 -30 -30>, angularTarget=<0 0 0>}
      localJointOffset=<-1 0 0>
     }
    ]
    relativeSpaceBone {boneName=spine_05}
    sphericalLimits [
     {limitRadius=13, drivingBone {boneName=pelvis}, sphereLocalOffset=<2 -5 6>}
     {limitRadius=11, drivingBone {boneName=spine_02}, sphereLocalOffset=<0 0 4>}
     {limitRadius=13, drivingBone {boneName=pelvis}, sphereLocalOffset=<2 -5 -6>}
     {limitRadius=11, drivingBone {boneName=spine_02}, sphereLocalOffset=<0 0 -4>}
    ]
   }
   "Anim Dynamics - Chain: dyn_hair_bk_04 -> dyn_hair_bk_05"!(
    ComponentPose=@n2
    Alpha=1
    ComponentLinearAccScale:Vector="0.5,0.5,0.5"
    ComponentLinearVelScale:Vector="0.5,0.5,0.5"
    ExternalForce="0,0,0"
   ) -> (Pose) {
    angularDampingOverride=0.999
    angularSpringConstant=200
    bAngularSpring=true
    bChain=true
    bOverrideAngularDamping=true
    bUseGravityOverride=true
    bUseSphericalLimits=true
    simulationSpace=RootRelative
    alphaBoolBlend {blendOption=Linear}
    boundBone {boneName=dyn_hair_bk_04}
    chainEnd {boneName=dyn_hair_bk_05}
    componentAppliedLinearAccClamp=<2000 2000 2000>
    componentLinearAccScale=<0.5 0.5 0.5>
    componentLinearVelScale=<0.5 0.5 0.5>
    gravityOverride {z=-980}
    physicsBodyDefinitions [
     {
      boundBone {boneName=dyn_hair_bk_04}
      boxExtents=<30 30 30>
      constraintSetup {angularLimitsMax=<0 10 20>, angularLimitsMin=<0 -10 -20>, angularTarget=<0 0 0>}
      localJointOffset=<-1 0 0>
     }
     {
      boundBone {boneName=dyn_hair_bk_05}
      boxExtents=<30 30 30>
      constraintSetup {angularLimitsMax=<0 30 30>, angularLimitsMin=<0 -30 -30>, angularTarget=<0 0 0>}
      localJointOffset=<-1 0 0>
     }
    ]
    relativeSpaceBone {boneName=spine_05}
    sphericalLimits [
     {limitRadius=13, drivingBone {boneName=pelvis}, sphereLocalOffset=<2 -5 6>}
     {limitRadius=11, drivingBone {boneName=spine_02}, sphereLocalOffset=<0 0 4>}
     {limitRadius=13, drivingBone {boneName=pelvis}, sphereLocalOffset=<2 -5 -6>}
     {limitRadius=11, drivingBone {boneName=spine_02}, sphereLocalOffset=<0 0 -4>}
    ]
   }
   n3 = "Transform (Modify) Bone - Bone: dyn_hair_l_MSTR"!(ComponentPose=Local To Component(LocalPose=Use cached pose 'StartPose'()), Alpha=1, Translation="0,-1,0", Rotation="0,-10,0") -> (Pose) {
    rotationMode=BMM_Additive
    rotationSpace=BCS_BoneSpace
    translationMode=BMM_Additive
    translationSpace=BCS_BoneSpace
    alphaBoolBlend {blendOption=Linear}
    boneToModify {boneName=dyn_hair_l_MSTR}
    rotation {yaw=-10}
    translation {y=-1}
   }
   n4 = "Transform (Modify) Bone - Bone: dyn_hair_bk_MSTR"!(ComponentPose=@n3, Alpha=1, Translation="0,-1,0", Rotation="0,-10,0") -> (Pose) {
    rotationMode=BMM_Additive
    rotationSpace=BCS_BoneSpace
    translationMode=BMM_Additive
    translationSpace=BCS_BoneSpace
    alphaBoolBlend {blendOption=Linear}
    boneToModify {boneName=dyn_hair_bk_MSTR}
    rotation {yaw=-10}
    translation {y=-1}
   }
   n5 = "Transform (Modify) Bone - Bone: dyn_hair_r_MSTR"!(ComponentPose=@n4, Alpha=1, Translation="0,-1,0", Rotation="0,-10,0") -> (Pose) {
    rotationMode=BMM_Additive
    rotationSpace=BCS_BoneSpace
    translationMode=BMM_Additive
    translationSpace=BCS_BoneSpace
    alphaBoolBlend {blendOption=Linear}
    boneToModify {boneName=dyn_hair_r_MSTR}
    rotation {yaw=-10}
    translation {y=-1}
   }
   n6 = RigidBody!(
    ComponentPose=@n5
    Alpha
    OverrideWorldGravity:Vector=Get Override World Gravity() -> (Override World Gravity:Vector)
    ExternalForce
    ComponentLinearAccScale:Vector=Get Component Linear Acc Scale() -> (Component Linear Acc Scale:Vector)
   ) -> (Pose) {
    bClampLinearTranslationLimitToRefPose=true
    bOverrideWorldGravity=true
    bTransferBoneVelocities=true
    overridePhysicsAsset=/Game/Characters/Player/Female/Medium/Bodies/F_MED_IchorIncisor/Meshes/Parts/F_MED_IchorIncisor_FaceAcc_PhysicS_InSky
    simulationSpace=BaseBoneSpace
    alphaBoolBlend {blendOption=Linear}
    alphaScaleBiasClamp {interpSpeedDecreasing=4, interpSpeedIncreasing=6}
    baseBoneRef {boneName=neck_01}
    componentAppliedLinearAccClamp=<800 800 500>
    componentLinearAccScale=<0.6 0.6 0.2>
    componentLinearVelScale=<0.5 0.5 0.5>
    overrideWorldGravity {z=-875}
    simSpaceSettings {maxAngularAcceleration=200, maxAngularVelocity=5, maxLinearAcceleration=500, maxLinearVelocity=200}
   }
   n7 = Blend Poses by bool(
    BlendPose_0=Component To Local(ComponentPose=@n6)
    BlendPose_1=Use cached pose 'RB_Pose'()
    BlendTime_0=0.3
    BlendTime_1=0.2
    bActiveValue=Get bIsSkyDiving()
   ) {blendType=Linear}
   n8 = "Constraint - dyn_hair_bk_01 (P)"(ComponentPose=Local To Component(LocalPose=Use cached pose 'RB_Pose'()), Alpha=1, ConstraintWeights_0=1) {
    alphaBoolBlend {blendOption=Linear}
    boneToModify {boneName=dyn_hair_bk_01}
    constraintSetup [{targetBone {boneName=head}}]
    constraintWeights [1]
   }
   n9 = "Constraint - dyn_hair_l_01 (P)"(ComponentPose=@n8, Alpha=1, ConstraintWeights_0=1) {
    alphaBoolBlend {blendOption=Linear}
    boneToModify {boneName=dyn_hair_l_01}
    constraintSetup [{targetBone {boneName=head}}]
    constraintWeights [1]
   }
   n10 = "Constraint - dyn_hair_r_01 (P)"(ComponentPose=@n9, Alpha=1, ConstraintWeights_0=1) {
    alphaBoolBlend {blendOption=Linear}
    boneToModify {boneName=dyn_hair_r_01}
    constraintSetup [{targetBone {boneName=head}}]
    constraintWeights [1]
   }
   n11 = "Constraint - dyn_hair_bk_02 (P)"(ComponentPose=@n10, Alpha=0.5, ConstraintWeights_0=1) {
    alphaBoolBlend {blendOption=Linear}
    boneToModify {boneName=dyn_hair_bk_02}
    constraintSetup [{targetBone {boneName=spine_05}}]
    constraintWeights [1]
   }
   n12 = "Constraint - dyn_hair_l_02 (P)"(ComponentPose=@n11, Alpha=0.5, ConstraintWeights_0=1) {
    alphaBoolBlend {blendOption=Linear}
    boneToModify {boneName=dyn_hair_l_02}
    constraintSetup [{targetBone {boneName=spine_05}}]
    constraintWeights [1]
   }
   n13 = "Constraint - dyn_hair_r_02 (P)"(ComponentPose=@n12, Alpha=0.5, ConstraintWeights_0=1) {
    alphaBoolBlend {blendOption=Linear}
    boneToModify {boneName=dyn_hair_r_02}
    constraintSetup [{targetBone {boneName=spine_05}}]
    constraintWeights [1]
   }
   n14 = "Constraint - dyn_hair_bk_03 (P)"(ComponentPose=@n13, Alpha=0.5, ConstraintWeights_0=1) {
    alphaBoolBlend {blendOption=Linear}
    boneToModify {boneName=dyn_hair_bk_03}
    constraintSetup [{targetBone {boneName=spine_05}}]
    constraintWeights [1]
   }
   n15 = "Constraint - dyn_hair_l_03 (P)"(ComponentPose=@n14, Alpha=0.5, ConstraintWeights_0=1) {
    alphaBoolBlend {blendOption=Linear}
    boneToModify {boneName=dyn_hair_l_03}
    constraintSetup [{targetBone {boneName=spine_05}}]
    constraintWeights [1]
   }
   n16 = "Constraint - dyn_hair_r_03 (P)"(ComponentPose=@n15, Alpha=0.5, ConstraintWeights_0=1) {
    alphaBoolBlend {blendOption=Linear}
    boneToModify {boneName=dyn_hair_r_03}
    constraintSetup [{targetBone {boneName=spine_05}}]
    constraintWeights [1]
   }
   n17 = "Constraint - dyn_hair_bk_04 (P)"(ComponentPose=@n16, Alpha=0.5, ConstraintWeights_0=1) {
    alphaBoolBlend {blendOption=Linear}
    boneToModify {boneName=dyn_hair_bk_04}
    constraintSetup [{targetBone {boneName=spine_04}}]
    constraintWeights [1]
   }
   n18 = "Constraint - dyn_hair_l_05 (P)"(ComponentPose=@n17, Alpha=0.5, ConstraintWeights_0=1) {
    alphaBoolBlend {blendOption=Linear}
    boneToModify {boneName=dyn_hair_l_05}
    constraintSetup [{targetBone {boneName=spine_04}}]
    constraintWeights [1]
   }
   n19 = "Constraint - dyn_hair_r_05 (P)"(ComponentPose=@n18, Alpha=0.5, ConstraintWeights_0=1) {
    alphaBoolBlend {blendOption=Linear}
    boneToModify {boneName=dyn_hair_r_05}
    constraintSetup [{targetBone {boneName=spine_04}}]
    constraintWeights [1]
   }
   n20 = Component To Local(ComponentPose=@n19)
   n21 = Get bIsSkydiveDiveMode()
   Blend Poses by bool(
    BlendPose_0=Blend Poses by bool(
     BlendPose_0=Use cached pose 'WithBackpack_Pose'()
     BlendPose_1=Blend Poses by bool(BlendPose_0=@n20, BlendPose_1=@n7, BlendTime_0=1, BlendTime_1=0.2, bActiveValue=@n21) {blendType=Linear}
     BlendTime_0=0.2
     BlendTime_1=0.2
     bActiveValue=Get Wearing_Backpack() -> (Wearing_Backpack:bool)
    ) {blendType=Linear}
    BlendPose_1=Use cached pose 'PreSimPose'()
    BlendTime_0=0.2
    BlendTime_1=0.2
    bActiveValue=Get bRigidBodyEnabled()
   ) {blendType=Linear}
   n22 = Get Head_Rot_Bk() -> (Head_Rot_Bk:double)
   n23 = "Transform (Modify) Bone - Bone: dyn_hair_l_MSTR"!(ComponentPose, Alpha=@n22, Translation="0,-3,0", Rotation="0,0,0", Scale:Vector="1,1,1") -> (Pose) {
    rotationMode=BMM_Additive
    rotationSpace=BCS_BoneSpace
    translationMode=BMM_Additive
    translationSpace=BCS_BoneSpace
    alphaBoolBlend {blendOption=Linear}
    boneToModify {boneName=dyn_hair_l_MSTR}
    translation {y=-3}
   }
   n24 = "Transform (Modify) Bone - Bone: dyn_hair_bk_MSTR"!(ComponentPose=@n23, Alpha=@n22, Translation="0,-2,0", Rotation="0,0,0", Scale:Vector="1,1,1") -> (Pose) {
    rotationMode=BMM_Additive
    rotationSpace=BCS_BoneSpace
    translationMode=BMM_Additive
    translationSpace=BCS_BoneSpace
    alphaBoolBlend {blendOption=Linear}
    boneToModify {boneName=dyn_hair_bk_MSTR}
    translation {y=-2}
   }
   n25 = "Transform (Modify) Bone - Bone: dyn_hair_r_MSTR"!(ComponentPose=@n24, Alpha=@n22, Translation="0,-2,0", Rotation="0,0,0", Scale:Vector="1,1,1") -> (Pose) {
    rotationMode=BMM_Additive
    rotationSpace=BCS_BoneSpace
    translationMode=BMM_Additive
    translationSpace=BCS_BoneSpace
    alphaBoolBlend {blendOption=Linear}
    boneToModify {boneName=dyn_hair_r_MSTR}
    translation {y=-2}
   }
   n26 = Get Head_Rot_R() -> (Head_Rot_R:double)
   n27 = "Transform (Modify) Bone - Bone: dyn_hair_r_MSTR"(ComponentPose=@n25, Alpha=@n26, Translation="0,1,0", Rotation="0,0,0", Scale="1,1,1") {
    rotationMode=BMM_Additive
    rotationSpace=BCS_BoneSpace
    translationMode=BMM_Additive
    translationSpace=BCS_BoneSpace
    alphaBoolBlend {blendOption=Linear}
    boneToModify {boneName=dyn_hair_r_MSTR}
    translation {y=1}
   }
   n28 = "Transform (Modify) Bone - Bone: dyn_hair_bk_MSTR"(ComponentPose=@n27, Alpha=@n26, Translation="-2,-2,1", Rotation="-20,0,-20", Scale="1,1,1") {
    rotationMode=BMM_Additive
    rotationSpace=BCS_BoneSpace
    translationMode=BMM_Additive
    translationSpace=BCS_BoneSpace
    alphaBoolBlend {blendOption=Linear}
    boneToModify {boneName=dyn_hair_bk_MSTR}
    rotation {pitch=-20, roll=-20}
    translation=<-2 -2 1>
   }
   "Transform (Modify) Bone - Bone: dyn_hair_l_MSTR"(ComponentPose=@n28, Alpha=@n26, Translation="-5,-4,2", Rotation="0,0,0", Scale="1,1,1") {
    rotationMode=BMM_Additive
    rotationSpace=BCS_BoneSpace
    translationMode=BMM_Additive
    translationSpace=BCS_BoneSpace
    alphaBoolBlend {blendOption=Linear}
    boneToModify {boneName=dyn_hair_l_MSTR}
    translation=<-5 -4 2>
   }
   Output Pose(Result=Use cached pose 'Post RBAN Hair'()) {layerGroup=None}
   n29 = RigidBody(
    ComponentPose=Local To Component(LocalPose=Use cached pose 'Backpack_Tranform_Pose'())
    OverrideWorldGravity=Get Override World Gravity()
    ComponentLinearAccScale=Get Component Linear Acc Scale()
   ) {
    bClampLinearTranslationLimitToRefPose=true
    bOverrideWorldGravity=true
    bTransferBoneVelocities=true
    overridePhysicsAsset=/Game/Characters/Player/Female/Medium/Bodies/F_MED_IchorIncisor/Meshes/Parts/F_MED_IchorIncisor_FaceAcc_Physics_WithBack
    simulationSpace=BaseBoneSpace
    alphaBoolBlend {blendOption=Linear}
    alphaScaleBiasClamp {interpSpeedDecreasing=4, interpSpeedIncreasing=6}
    baseBoneRef {boneName=neck_01}
    componentAppliedLinearAccClamp=<800 800 500>
    componentLinearAccScale=<0.6 0.6 0.2>
    componentLinearVelScale=<0.5 0.5 0.5>
    overrideWorldGravity {z=-875}
    simSpaceSettings {maxAngularAcceleration=200, maxAngularVelocity=5, maxLinearAcceleration=500, maxLinearVelocity=200}
   }
   n30 = "Transform (Modify) Bone - Bone: dyn_hair_l_05"!(ComponentPose=Local To Component(LocalPose=Use cached pose 'Backpack_Tranform_Pose'()), Alpha=1, Translation="-2,1,0", Rotation="0,5,0") -> (Pose) {
    rotationMode=BMM_Additive
    rotationSpace=BCS_BoneSpace
    translationMode=BMM_Additive
    translationSpace=BCS_BoneSpace
    alphaBoolBlend {blendOption=Linear}
    boneToModify {boneName=dyn_hair_l_05}
    rotation {yaw=5}
    translation=<-2 1>
   }
   n31 = "Transform (Modify) Bone - Bone: dyn_hair_bk_04"!(ComponentPose=@n30, Alpha=1, Translation="-2,1,0", Rotation="0,5,0") -> (Pose) {
    rotationMode=BMM_Additive
    rotationSpace=BCS_BoneSpace
    translationMode=BMM_Additive
    translationSpace=BCS_BoneSpace
    alphaBoolBlend {blendOption=Linear}
    boneToModify {boneName=dyn_hair_bk_04}
    rotation {yaw=5}
    translation=<-2 1>
   }
   n32 = "Transform (Modify) Bone - Bone: dyn_hair_r_05"!(ComponentPose=@n31, Alpha=1, Translation="-2,1,0", Rotation="0,5,0") -> (Pose) {
    rotationMode=BMM_Additive
    rotationSpace=BCS_BoneSpace
    translationMode=BMM_Additive
    translationSpace=BCS_BoneSpace
    alphaBoolBlend {blendOption=Linear}
    boneToModify {boneName=dyn_hair_r_05}
    rotation {yaw=5}
    translation=<-2 1>
   }
   n33 = RigidBody(ComponentPose=@n32, OverrideWorldGravity=Get Override World Gravity(), ComponentLinearAccScale=Get Component Linear Acc Scale()) {
    bClampLinearTranslationLimitToRefPose=true
    bOverrideWorldGravity=true
    bTransferBoneVelocities=true
    overridePhysicsAsset=&F_MED_IchorIncisor_FaceAcc_Physics_WithBack
    simulationSpace=BaseBoneSpace
    alphaBoolBlend {blendOption=Linear}
    alphaScaleBiasClamp {interpSpeedDecreasing=4, interpSpeedIncreasing=6}
    baseBoneRef {boneName=neck_01}
    componentAppliedLinearAccClamp=<800 800 500>
    componentLinearAccScale=<0.6 0.6 0.2>
    componentLinearVelScale=<0.5 0.5 0.5>
    overrideWorldGravity {z=-875}
    simSpaceSettings {maxAngularAcceleration=200, maxAngularVelocity=5, maxLinearAcceleration=500, maxLinearVelocity=200}
   }
   n34 = "Constraint - dyn_hair_bk_05 (P)"(ComponentPose=@n33, Alpha=1, ConstraintWeights_0=1) {
    alphaBoolBlend {blendOption=Linear}
    boneToModify {boneName=dyn_hair_bk_05}
    constraintSetup [{targetBone {boneName=dyn_hair_bk_04}}]
    constraintWeights [1]
   }
   n35 = "Constraint - dyn_hair_l_06 (P)"(ComponentPose=@n34, Alpha=1, ConstraintWeights_0=1) {
    alphaBoolBlend {blendOption=Linear}
    boneToModify {boneName=dyn_hair_l_06}
    constraintSetup [{targetBone {boneName=dyn_hair_l_05}}]
    constraintWeights [1]
   }
   n36 = "Constraint - dyn_hair_r_06 (P)"(ComponentPose=@n35, Alpha=1, ConstraintWeights_0=1) {
    alphaBoolBlend {blendOption=Linear}
    boneToModify {boneName=dyn_hair_r_06}
    constraintSetup [{targetBone {boneName=dyn_hair_r_05}}]
    constraintWeights [1]
   }
   Save cached pose 'WithBackpack_Pose'(Pose=Blend Poses by bool(
     BlendPose_0=Component To Local(ComponentPose=@n36)
     BlendPose_1=Component To Local(ComponentPose=@n29)
     BlendTime_0=0.7
     bActiveValue=Get SimpleJump() -> (SimpleJump:bool)
    ) {blendType=Linear}) {cacheName=WithBackpack_Pose}
   Save cached pose 'RB_Pose'() {cacheName=RB_Pose}
   n37 = Control Rig!(Source=Copy Pose From Mesh() {bUseAttachedParent=true}, Alpha, WearingBackpack:bool=Get Wearing_Backpack()) -> (Pose) {
    bTransferInputCurves=false
    alphaBoolBlend {blendOption=Linear}
    controlRigAssetReference {blueprintRigClass=/Game/Characters/Player/Female/Medium/Bodies/F_MED_IchorIncisor/Meshes/Parts/F_MED_IchorIncisor_FaceAcc_CtrlRig.F_MED_IchorIncisor_FaceAcc_CtrlRig_C}
   }
   Save cached pose 'StartPose'(Pose=@n37) {cacheName=StartPose}
   n38 = "Constraint - dyn_piercing (P)"(ComponentPose=Local To Component(LocalPose=Use cached pose 'StartPose'()), Alpha=1, ConstraintWeights_0=1) {
    alphaBoolBlend {blendOption=Linear}
    boneToModify {boneName=dyn_piercing}
    constraintSetup [{targetBone {boneName=FACIAL_C_LipLower}}]
    constraintWeights [1]
   }
   n39 = "Transform (Modify) Bone - Bone: dyn_piercing"!(ComponentPose=@n38, Alpha=1, Translation="0,0,0.08", Rotation="0,0,0") -> (Pose) {
    rotationMode=BMM_Additive
    rotationSpace=BCS_BoneSpace
    translationMode=BMM_Additive
    translationSpace=BCS_BoneSpace
    alphaBoolBlend {blendOption=Linear}
    boneToModify {boneName=dyn_piercing}
    translation {z=0.08}
   }
   n40 = "Constraint - dyn_strand_long_r_01 (PP)"!(ComponentPose=@n39, Alpha=1, ConstraintWeights_0=0.6, ConstraintWeights_1:float=0.4) -> (Pose) {
    alphaBoolBlend {blendOption=Linear}
    boneToModify {boneName=dyn_strand_long_r_01}
    constraintSetup [{targetBone {boneName=head}}, {targetBone {boneName=spine_05}}]
    constraintWeights [1,1]
   }
   n41 = "Constraint - dyn_stland_long_l_01 (PP)"!(ComponentPose=@n40, Alpha=1, ConstraintWeights_0=0.6, ConstraintWeights_1:float=0.4) -> (Pose) {
    alphaBoolBlend {blendOption=Linear}
    boneToModify {boneName=dyn_stland_long_l_01}
    constraintSetup [{targetBone {boneName=head}}, {targetBone {boneName=spine_05}}]
    constraintWeights [1,1]
   }
   n42 = "Constraint - dyn_strand_long_r_02 (PPP)"!(ComponentPose=@n41, Alpha=1, ConstraintWeights_0=0.2, ConstraintWeights_1:float=0.3, ConstraintWeights_2:float=0.5) -> (Pose) {
    alphaBoolBlend {blendOption=Linear}
    boneToModify {boneName=dyn_strand_long_r_02}
    constraintSetup [{targetBone {boneName=spine_05}}, {targetBone {boneName=head}}, {targetBone {boneName=clavicle_r}}]
    constraintWeights [1,1,1]
   }
   n43 = "Constraint - dyn_stland_long_l_02 (PPP)"!(ComponentPose=@n42, Alpha=1, ConstraintWeights_0=0.7, ConstraintWeights_1:float=0.3, ConstraintWeights_2:float=0) -> (Pose) {
    alphaBoolBlend {blendOption=Linear}
    boneToModify {boneName=dyn_stland_long_l_02}
    constraintSetup [{targetBone {boneName=head}}, {targetBone {boneName=spine_05}}, {targetBone {boneName=clavicle_l}}]
    constraintWeights [1,1,1]
   }
   n44 = "Constraint - dyn_strand_long_r_03 (PP)"!(ComponentPose=@n43, Alpha=1, ConstraintWeights_0=0.2, ConstraintWeights_1:float=0.8) -> (Pose) {
    alphaBoolBlend {blendOption=Linear}
    boneToModify {boneName=dyn_strand_long_r_03}
    constraintSetup [{targetBone {boneName=spine_05}}, {targetBone {boneName=clavicle_r}}]
    constraintWeights [1,1]
   }
   n45 = "Constraint - dyn_stland_long_l_03 (PP)"!(ComponentPose=@n44, Alpha=1, ConstraintWeights_0=0.3, ConstraintWeights_1:float=0.7) -> (Pose) {
    alphaBoolBlend {blendOption=Linear}
    boneToModify {boneName=dyn_stland_long_l_03}
    constraintSetup [{targetBone {boneName=spine_05}}, {targetBone {boneName=clavicle_l}}]
    constraintWeights [1,1]
   }
   n46 = Anim Dynamics - Bone: dyn_fringe_r_01(ComponentPose=@n45, Alpha=1, ExternalForce="0,0,0") {
    angularDampingOverride=0.999
    angularSpringConstant=500
    bAngularSpring=true
    bOverrideAngularDamping=true
    bUseGravityOverride=true
    gravityScale=-1
    simulationSpace=RootRelative
    alphaBoolBlend {blendOption=Linear}
    boundBone {boneName=dyn_fringe_r_01}
    componentAppliedLinearAccClamp=<3000 3000 3000>
    componentLinearAccScale=<0.3 0.3 0.3>
    componentLinearVelScale=<0.3 0.3 0.3>
    gravityOverride {z=-980}
    physicsBodyDefinitions [{
      boundBone {boneName=dyn_fringe_r_01}
      boxExtents=<7 7 7>
      constraintSetup {angularLimitsMax=<5 10 20>, angularLimitsMin=<-5 -10 -10>, angularTarget=<0 0 0>}
      localJointOffset=<1 0 0>
     }]
   }
   n47 = Anim Dynamics - Bone: dyn_fringe_l_01(ComponentPose=@n46, Alpha=1, ExternalForce="0,0,0") {
    angularDampingOverride=0.999
    angularSpringConstant=500
    bAngularSpring=true
    bOverrideAngularDamping=true
    bUseGravityOverride=true
    gravityScale=-1
    simulationSpace=RootRelative
    alphaBoolBlend {blendOption=Linear}
    boundBone {boneName=dyn_fringe_l_01}
    componentAppliedLinearAccClamp=<3000 3000 3000>
    componentLinearAccScale=<0.3 0.3 0.3>
    componentLinearVelScale=<0.3 0.3 0.3>
    gravityOverride {z=-980}
    physicsBodyDefinitions [{
      boundBone {boneName=dyn_fringe_l_01}
      boxExtents=<7 7 7>
      constraintSetup {angularLimitsMax=<5 10 20>, angularLimitsMin=<-5 -10 -10>, angularTarget=<0 0 0>}
      localJointOffset=<1 0 0>
     }]
   }
   n48 = Anim Dynamics - Bone: dyn_fringe_fr_01(ComponentPose=@n47, Alpha=1, ExternalForce="0,0,0") {
    angularDampingOverride=0.999
    angularSpringConstant=500
    bAngularSpring=true
    bOverrideAngularDamping=true
    bUseGravityOverride=true
    gravityScale=-1
    simulationSpace=RootRelative
    alphaBoolBlend {blendOption=Linear}
    boundBone {boneName=dyn_fringe_fr_01}
    componentAppliedLinearAccClamp=<3000 3000 3000>
    componentLinearAccScale=<0.3 0.3 0.3>
    componentLinearVelScale=<0.3 0.3 0.3>
    gravityOverride {z=-980}
    physicsBodyDefinitions [{
      boundBone {boneName=dyn_fringe_fr_01}
      boxExtents=<7 5 7>
      constraintSetup {angularLimitsMax=<5 10 30>, angularLimitsMin=<-5 -10 -10>, angularTarget=<0 0 0>}
      localJointOffset=<1 0 0>
     }]
   }
   n49 = Anim Dynamics - Bone: dyn_strand_short_l_01(ComponentPose=@n48, Alpha=1, ExternalForce="0,0,0") {
    angularDampingOverride=0.9
    angularSpringConstant=500
    bAngularSpring=true
    bOverrideAngularDamping=true
    bUseGravityOverride=true
    gravityScale=-1
    simulationSpace=RootRelative
    alphaBoolBlend {blendOption=Linear}
    boundBone {boneName=dyn_strand_short_l_01}
    componentAppliedLinearAccClamp=<3000 3000 3000>
    componentLinearAccScale=<0.3 0.3 0.3>
    componentLinearVelScale=<0.3 0.3 0.3>
    gravityOverride {z=-980}
    physicsBodyDefinitions [{
      boundBone {boneName=dyn_strand_short_l_01}
      boxExtents=<15 15 5>
      constraintSetup {angularLimitsMax=<5 0 10>, angularLimitsMin=<-5 0 -10>, angularTarget=<0 0 0>}
      localJointOffset=<0.5 0 0>
     }]
   }
   n50 = Anim Dynamics - Bone: dyn_strand_short_l_02(ComponentPose=@n49, Alpha=1, ExternalForce="0,0,0") {
    angularDampingOverride=0.9
    angularSpringConstant=500
    bAngularSpring=true
    bOverrideAngularDamping=true
    bUseGravityOverride=true
    gravityScale=-1
    simulationSpace=RootRelative
    alphaBoolBlend {blendOption=Linear}
    boundBone {boneName=dyn_strand_short_l_02}
    componentAppliedLinearAccClamp=<3000 3000 3000>
    componentLinearAccScale=<0.3 0.3 0.3>
    componentLinearVelScale=<0.3 0.3 0.3>
    gravityOverride {z=-980}
    physicsBodyDefinitions [{
      boundBone {boneName=dyn_strand_short_l_02}
      boxExtents=<12 12 5>
      constraintSetup {angularLimitsMax=<5 10 10>, angularLimitsMin=<-5 -10 -10>, angularTarget=<0 0 0>}
      localJointOffset=<0.5 0 0>
     }]
   }
   n51 = Anim Dynamics - Bone: dyn_strand_short_l_03(ComponentPose=@n50, Alpha=1, ExternalForce="0,0,0") {
    angularDampingOverride=0.999
    angularSpringConstant=500
    bAngularSpring=true
    bOverrideAngularDamping=true
    bUseGravityOverride=true
    gravityScale=-1
    simulationSpace=RootRelative
    alphaBoolBlend {blendOption=Linear}
    boundBone {boneName=dyn_strand_short_l_03}
    componentAppliedLinearAccClamp=<3000 3000 3000>
    componentLinearAccScale=<0.3 0.3 0.3>
    componentLinearVelScale=<0.3 0.3 0.3>
    gravityOverride {z=-980}
    physicsBodyDefinitions [{
      boundBone {boneName=dyn_strand_short_l_03}
      boxExtents=<7 5 7>
      constraintSetup {angularLimitsMax=<5 10 30>, angularLimitsMin=<-5 -10 10>, angularTarget=<0 0 0>}
      localJointOffset=<2 0 0>
     }]
   }
   n52 = Anim Dynamics - Bone: dyn_stland_short_r_01(ComponentPose=@n51, Alpha=1, ExternalForce="0,0,0") {
    angularDampingOverride=0.9
    angularSpringConstant=500
    bAngularSpring=true
    bOverrideAngularDamping=true
    bUseGravityOverride=true
    gravityScale=-1
    simulationSpace=RootRelative
    alphaBoolBlend {blendOption=Linear}
    boundBone {boneName=dyn_stland_short_r_01}
    componentAppliedLinearAccClamp=<3000 3000 3000>
    componentLinearAccScale=<0.3 0.3 0.3>
    componentLinearVelScale=<0.3 0.3 0.3>
    gravityOverride {z=-980}
    physicsBodyDefinitions [{
      boundBone {boneName=dyn_stland_short_r_01}
      boxExtents=<15 15 5>
      constraintSetup {angularLimitsMax=<5 0 10>, angularLimitsMin=<-5 0 -10>, angularTarget=<0 0 0>}
      localJointOffset=<0.5 0 0>
     }]
   }
   n53 = Anim Dynamics - Bone: dyn_stland_short_r_02(ComponentPose=@n52, Alpha=1, ExternalForce="0,0,0") {
    angularDampingOverride=0.9
    angularSpringConstant=500
    bAngularSpring=true
    bOverrideAngularDamping=true
    bUseGravityOverride=true
    gravityScale=-1
    simulationSpace=RootRelative
    alphaBoolBlend {blendOption=Linear}
    boundBone {boneName=dyn_stland_short_r_02}
    componentAppliedLinearAccClamp=<3000 3000 3000>
    componentLinearAccScale=<0.3 0.3 0.3>
    componentLinearVelScale=<0.3 0.3 0.3>
    gravityOverride {z=-980}
    physicsBodyDefinitions [{
      boundBone {boneName=dyn_stland_short_r_02}
      boxExtents=<12 12 5>
      constraintSetup {angularLimitsMax=<5 10 10>, angularLimitsMin=<-5 -10 -10>, angularTarget=<0 0 0>}
      localJointOffset=<0.5 0 0>
     }]
   }
   n54 = Anim Dynamics - Bone: dyn_stland_short_r_03(ComponentPose=@n53, Alpha=1, ExternalForce="0,0,0") {
    angularDampingOverride=0.999
    angularSpringConstant=500
    bAngularSpring=true
    bOverrideAngularDamping=true
    bUseGravityOverride=true
    gravityScale=-1
    simulationSpace=RootRelative
    alphaBoolBlend {blendOption=Linear}
    boundBone {boneName=dyn_stland_short_r_03}
    componentAppliedLinearAccClamp=<3000 3000 3000>
    componentLinearAccScale=<0.3 0.3 0.3>
    componentLinearVelScale=<0.3 0.3 0.3>
    gravityOverride {z=-980}
    physicsBodyDefinitions [{
      boundBone {boneName=dyn_stland_short_r_03}
      boxExtents=<7 5 7>
      constraintSetup {angularLimitsMax=<5 10 30>, angularLimitsMin=<-5 -10 10>, angularTarget=<0 0 0>}
      localJointOffset=<2 0 0>
     }]
   }
   Save cached pose 'PreSimPose'(Pose=Component To Local(ComponentPose=@n54)) {cacheName=PreSimPose}
   n55 = "Transform (Modify) Bone - Bone: dyn_hair_l_05"(ComponentPose=Local To Component(LocalPose=Use cached pose 'Backpack_Base_Pose'()), Alpha=1, Translation="0,-1,0", Rotation="0,-15,0") {
    rotationMode=BMM_Additive
    rotationSpace=BCS_BoneSpace
    translationMode=BMM_Additive
    translationSpace=BCS_BoneSpace
    alphaBoolBlend {blendOption=Linear}
    boneToModify {boneName=dyn_hair_l_05}
    rotation {yaw=-15}
    translation {y=-1}
   }
   n56 = "Transform (Modify) Bone - Bone: dyn_hair_bk_04"(ComponentPose=@n55, Alpha=1, Translation="0,-1,0", Rotation="0,-15,0") {
    rotationMode=BMM_Additive
    rotationSpace=BCS_BoneSpace
    translationMode=BMM_Additive
    translationSpace=BCS_BoneSpace
    alphaBoolBlend {blendOption=Linear}
    boneToModify {boneName=dyn_hair_bk_04}
    rotation {yaw=-15}
    translation {y=-1}
   }
   n57 = "Transform (Modify) Bone - Bone: dyn_hair_r_05"(ComponentPose=@n56, Alpha=1, Translation="0,-1,0", Rotation="0,-15,0") {
    rotationMode=BMM_Additive
    rotationSpace=BCS_BoneSpace
    translationMode=BMM_Additive
    translationSpace=BCS_BoneSpace
    alphaBoolBlend {blendOption=Linear}
    boneToModify {boneName=dyn_hair_r_05}
    rotation {yaw=-15}
    translation {y=-1}
   }
   n58 = "Transform (Modify) Bone - Bone: dyn_hair_l_05"(ComponentPose=Local To Component(LocalPose=Use cached pose 'Backpack_Base_Pose'()), Alpha=1, Translation="-5,-1,0", Rotation="0,0,0") {
    rotationMode=BMM_Additive
    rotationSpace=BCS_BoneSpace
    translationMode=BMM_Additive
    translationSpace=BCS_BoneSpace
    alphaBoolBlend {blendOption=Linear}
    boneToModify {boneName=dyn_hair_l_05}
    translation=<-5 -1>
   }
   n59 = "Transform (Modify) Bone - Bone: dyn_hair_bk_04"(ComponentPose=@n58, Alpha=1, Translation="-5,-1,0", Rotation="0,0,0") {
    rotationMode=BMM_Additive
    rotationSpace=BCS_BoneSpace
    translationMode=BMM_Additive
    translationSpace=BCS_BoneSpace
    alphaBoolBlend {blendOption=Linear}
    boneToModify {boneName=dyn_hair_bk_04}
    translation=<-5 -1>
   }
   n60 = "Transform (Modify) Bone - Bone: dyn_hair_r_05"(ComponentPose=@n59, Alpha=1, Translation="-5,-1,0", Rotation="0,0,0") {
    rotationMode=BMM_Additive
    rotationSpace=BCS_BoneSpace
    translationMode=BMM_Additive
    translationSpace=BCS_BoneSpace
    alphaBoolBlend {blendOption=Linear}
    boneToModify {boneName=dyn_hair_r_05}
    translation=<-5 -1>
   }
   n61 = Blend Poses by bool(
    BlendPose_0=Component To Local(ComponentPose=@n57)
    BlendPose_1=Component To Local(ComponentPose=@n60)
    BlendTime_0=0.2
    BlendTime_1=0.2
    bActiveValue=Get bIsParachuteOpened()
   ) {blendType=Linear}
   n62 = "Transform (Modify) Bone - Bone: dyn_hair_l_05"(ComponentPose=Local To Component(LocalPose=Use cached pose 'Backpack_Base_Pose'()), Alpha=1, Translation="0,-1,0", Rotation="10,-10,0") {
    rotationMode=BMM_Additive
    rotationSpace=BCS_BoneSpace
    translationMode=BMM_Additive
    translationSpace=BCS_BoneSpace
    alphaBoolBlend {blendOption=Linear}
    boneToModify {boneName=dyn_hair_l_05}
    rotation {pitch=10, yaw=-10}
    translation {y=-1}
   }
   n63 = "Transform (Modify) Bone - Bone: dyn_hair_bk_04"(ComponentPose=@n62, Alpha=1, Translation="0,-1,0", Rotation="0,-10,0") {
    rotationMode=BMM_Additive
    rotationSpace=BCS_BoneSpace
    translationMode=BMM_Additive
    translationSpace=BCS_BoneSpace
    alphaBoolBlend {blendOption=Linear}
    boneToModify {boneName=dyn_hair_bk_04}
    rotation {yaw=-10}
    translation {y=-1}
   }
   n64 = "Transform (Modify) Bone - Bone: dyn_hair_r_05"(ComponentPose=@n63, Alpha=1, Translation="0,-1,0", Rotation="-10,-10,0") {
    rotationMode=BMM_Additive
    rotationSpace=BCS_BoneSpace
    translationMode=BMM_Additive
    translationSpace=BCS_BoneSpace
    alphaBoolBlend {blendOption=Linear}
    boneToModify {boneName=dyn_hair_r_05}
    rotation {pitch=-10, yaw=-10}
    translation {y=-1}
   }
   n65 = Component To Local(ComponentPose=@n64)
   n66 = Get bIsSkydiveDiveMode()
   Save cached pose 'Backpack_Tranform_Pose'(Pose=Blend Poses by bool(
     BlendPose_0=Blend Poses by bool(BlendPose_0=@n65, BlendPose_1=@n61, BlendTime_0=0.2, BlendTime_1=0.2, bActiveValue=@n66) {blendType=Linear}
     BlendPose_1=Use cached pose 'Backpack_Base_Pose'()
     BlendTime_0=0.3
     BlendTime_1=0.3
     bActiveValue=Get bIsSkyDiving()
    ) {blendType=Linear}) {cacheName=Backpack_Tranform_Pose}
   n67 = "Transform (Modify) Bone - Bone: dyn_hair_l_05"(ComponentPose=Local To Component(LocalPose=Use cached pose 'PreSimPose'()), Alpha=1, Translation="1,1.5,1", Rotation="0,20,0") {
    rotationMode=BMM_Additive
    rotationSpace=BCS_BoneSpace
    translationMode=BMM_Additive
    translationSpace=BCS_BoneSpace
    alphaBoolBlend {blendOption=Linear}
    boneToModify {boneName=dyn_hair_l_05}
    rotation {yaw=20}
    translation=<1 1.5 1>
   }
   n68 = "Transform (Modify) Bone - Bone: dyn_hair_bk_04"(ComponentPose=@n67, Alpha=1, Translation="0,1.5,0", Rotation="0,20,0") {
    rotationMode=BMM_Additive
    rotationSpace=BCS_BoneSpace
    translationMode=BMM_Additive
    translationSpace=BCS_BoneSpace
    alphaBoolBlend {blendOption=Linear}
    boneToModify {boneName=dyn_hair_bk_04}
    rotation {yaw=20}
    translation {y=1.5}
   }
   n69 = "Transform (Modify) Bone - Bone: dyn_hair_r_05"(ComponentPose=@n68, Alpha=1, Translation="1,1.5,-1", Rotation="0,20,0") {
    rotationMode=BMM_Additive
    rotationSpace=BCS_BoneSpace
    translationMode=BMM_Additive
    translationSpace=BCS_BoneSpace
    alphaBoolBlend {blendOption=Linear}
    boneToModify {boneName=dyn_hair_r_05}
    rotation {yaw=20}
    translation=<1 1.5 -1>
   }
   n70 = Get Head_Rot_Bk()
   n71 = "Transform (Modify) Bone - Bone: dyn_hair_l_05"(ComponentPose=@n69, Alpha=@n70, Translation="0,-1,0", Rotation="0,-30,0") {
    rotationMode=BMM_Additive
    rotationSpace=BCS_BoneSpace
    translationMode=BMM_Additive
    translationSpace=BCS_BoneSpace
    alphaBoolBlend {blendOption=Linear}
    boneToModify {boneName=dyn_hair_l_05}
    rotation {yaw=-30}
    translation {y=-1}
   }
   n72 = "Transform (Modify) Bone - Bone: dyn_hair_bk_04"(ComponentPose=@n71, Alpha=@n70, Translation="0,-1,0", Rotation="0,-30,0") {
    rotationMode=BMM_Additive
    rotationSpace=BCS_BoneSpace
    translationMode=BMM_Additive
    translationSpace=BCS_BoneSpace
    alphaBoolBlend {blendOption=Linear}
    boneToModify {boneName=dyn_hair_bk_04}
    rotation {yaw=-30}
    translation {y=-1}
   }
   n73 = "Transform (Modify) Bone - Bone: dyn_hair_r_05"(ComponentPose=@n72, Alpha=@n70, Translation="0,-1,0", Rotation="0,-30,0") {
    rotationMode=BMM_Additive
    rotationSpace=BCS_BoneSpace
    translationMode=BMM_Additive
    translationSpace=BCS_BoneSpace
    alphaBoolBlend {blendOption=Linear}
    boneToModify {boneName=dyn_hair_r_05}
    rotation {yaw=-30}
    translation {y=-1}
   }
   Save cached pose 'Backpack_Base_Pose'(Pose=Component To Local(ComponentPose=@n73)) {cacheName=Backpack_Base_Pose}
   n74 = RigidBody!(ComponentPose=Local To Component(LocalPose=Use cached pose 'PreSimPose'()), Alpha, OverrideWorldGravity="0,0,-700", ExternalForce) -> (Pose) {
    bAllowPoseScalePropagation=false
    bOverrideWorldGravity=true
    overridePhysicsAsset=&F_MED_IchorIncisor_FaceAcc_Physics_WithBack
    simulationSpace=BaseBoneSpace
    alphaBoolBlend {blendOption=Linear}
    alphaScaleBiasClamp {interpSpeedDecreasing=4, interpSpeedIncreasing=6}
    baseBoneRef {boneName=spine_05}
    componentAppliedLinearAccClamp=<800 800 500>
    overrideWorldGravity {z=-875}
    simSpaceSettings {maxAngularAcceleration=30, maxAngularVelocity=8, maxLinearAcceleration=150, maxLinearVelocity=200, velocityScaleZ=0, worldAlpha=1}
   }
   n75 = RigidBody(ComponentPose=Local To Component(LocalPose=Use cached pose 'PreSimPose'()), OverrideWorldGravity="0,0,-700") {
    bAllowPoseScalePropagation=false
    bOverrideWorldGravity=true
    overridePhysicsAsset=/Game/Characters/Player/Female/Medium/Bodies/F_MED_IchorIncisor/Meshes/Parts/F_MED_IchorIncisor_FaceAcc_Physics2
    simulationSpace=BaseBoneSpace
    alphaBoolBlend {blendOption=Linear}
    alphaScaleBiasClamp {interpSpeedDecreasing=4, interpSpeedIncreasing=6}
    baseBoneRef {boneName=spine_05}
    componentAppliedLinearAccClamp=<800 800 500>
    componentLinearAccScale {y=0.2}
    componentLinearVelScale {y=0.2}
    overrideWorldGravity {z=-875}
    simSpaceSettings {maxAngularAcceleration=30, maxAngularVelocity=8, maxLinearAcceleration=150, maxLinearVelocity=200, worldAlpha=1, externalLinearDragV {y=0.3}}
   }
   Save cached pose 'Post RBAN Hair'(Pose=Blend Poses by bool(BlendPose_0=Component To Local(ComponentPose=@n74), BlendPose_1=Component To Local(ComponentPose=@n75), bActiveValue=Get Wearing_Backpack()) {blendType=Linear}) {cacheName=Post RBAN Hair}
   n76 = "Constraint - dyn_hair_bk_MSTR (P)"(Alpha=1, ConstraintWeights_0=1) {
    alphaBoolBlend {blendOption=Linear}
    boneToModify {boneName=dyn_hair_bk_MSTR}
    constraintSetup [{targetBone {boneName=head}}]
    constraintWeights [1]
   }
   n77 = "Constraint - dyn_hair_l_MSTR (P)"(ComponentPose=@n76, Alpha=1, ConstraintWeights_0=1) {
    alphaBoolBlend {blendOption=Linear}
    boneToModify {boneName=dyn_hair_l_MSTR}
    constraintSetup [{targetBone {boneName=head}}]
    constraintWeights [1]
   }
   n78 = "Constraint - dyn_hair_r_MSTR (P)"(ComponentPose=@n77, Alpha=1, ConstraintWeights_0=1) {
    alphaBoolBlend {blendOption=Linear}
    boneToModify {boneName=dyn_hair_r_MSTR}
    constraintSetup [{targetBone {boneName=head}}]
    constraintWeights [1]
   }
   n79 = "Constraint - dyn_hair_bk_01 (P)"(ComponentPose=@n78, Alpha=1, ConstraintWeights_0=1) {
    alphaBoolBlend {blendOption=Linear}
    boneToModify {boneName=dyn_hair_bk_01}
    constraintSetup [{targetBone {boneName=head}}]
    constraintWeights [1]
   }
   n80 = "Constraint - dyn_hair_l_01 (P)"(ComponentPose=@n79, Alpha=1, ConstraintWeights_0=1) {
    alphaBoolBlend {blendOption=Linear}
    boneToModify {boneName=dyn_hair_l_01}
    constraintSetup [{targetBone {boneName=head}}]
    constraintWeights [1]
   }
   n81 = "Constraint - dyn_hair_r_01 (P)"(ComponentPose=@n80, Alpha=1, ConstraintWeights_0=1) {
    alphaBoolBlend {blendOption=Linear}
    boneToModify {boneName=dyn_hair_r_01}
    constraintSetup [{targetBone {boneName=head}}]
    constraintWeights [1]
   }
   n82 = "Transform (Modify) Bone - Bone: dyn_hair_l_01"!(ComponentPose=@n81, Alpha=1, Translation="0,-1,0", Rotation="0,-20,0") -> (Pose) {
    rotationMode=BMM_Additive
    rotationSpace=BCS_BoneSpace
    translationMode=BMM_Additive
    translationSpace=BCS_BoneSpace
    alphaBoolBlend {blendOption=Linear}
    boneToModify {boneName=dyn_hair_l_01}
    rotation {yaw=-20}
    translation {y=-1}
   }
   n83 = "Transform (Modify) Bone - Bone: dyn_hair_bk_01"!(ComponentPose=@n82, Alpha=1, Translation="0,-1,0", Rotation="0,-20,0") -> (Pose) {
    rotationMode=BMM_Additive
    rotationSpace=BCS_BoneSpace
    translationMode=BMM_Additive
    translationSpace=BCS_BoneSpace
    alphaBoolBlend {blendOption=Linear}
    boneToModify {boneName=dyn_hair_bk_01}
    rotation {yaw=-20}
    translation {y=-1}
   }
   n84 = "Transform (Modify) Bone - Bone: dyn_hair_r_01"!(ComponentPose=@n83, Alpha=1, Translation="0,-1,0", Rotation="0,-20,0") -> (Pose) {
    rotationMode=BMM_Additive
    rotationSpace=BCS_BoneSpace
    translationMode=BMM_Additive
    translationSpace=BCS_BoneSpace
    alphaBoolBlend {blendOption=Linear}
    boneToModify {boneName=dyn_hair_r_01}
    rotation {yaw=-20}
    translation {y=-1}
   }
   n85 = "Transform (Modify) Bone - Bone: dyn_hair_l_01"(ComponentPose=@n84, Alpha=1, Translation="-3,-3,3", Rotation="0,-30,0") {
    rotationMode=BMM_Additive
    rotationSpace=BCS_BoneSpace
    translationMode=BMM_Additive
    translationSpace=BCS_BoneSpace
    alphaBoolBlend {blendOption=Linear}
    boneToModify {boneName=dyn_hair_l_01}
    rotation {yaw=-30}
    translation=<-3 -3 3>
   }
   n86 = "Transform (Modify) Bone - Bone: dyn_hair_bk_01"(ComponentPose=@n85, Alpha=1, Translation="0,0,2", Rotation="0,0,0") {
    rotationMode=BMM_Additive
    rotationSpace=BCS_BoneSpace
    translationMode=BMM_Additive
    translationSpace=BCS_BoneSpace
    alphaBoolBlend {blendOption=Linear}
    boneToModify {boneName=dyn_hair_bk_01}
    translation {z=2}
   }
   n87 = "Transform (Modify) Bone - Bone: dyn_hair_l_01"(ComponentPose=@n86, Alpha=1, Translation="0,-2,0", Rotation="0,0,0") {
    rotationMode=BMM_Additive
    rotationSpace=BCS_BoneSpace
    translationMode=BMM_Additive
    translationSpace=BCS_BoneSpace
    alphaBoolBlend {blendOption=Linear}
    boneToModify {boneName=dyn_hair_l_01}
    translation {y=-2}
   }
   n88 = "Transform (Modify) Bone - Bone: dyn_hair_bk_01"(ComponentPose=@n87, Alpha=1, Translation="0,-1,0", Rotation="0,0,0") {
    rotationMode=BMM_Additive
    rotationSpace=BCS_BoneSpace
    translationMode=BMM_Additive
    translationSpace=BCS_BoneSpace
    alphaBoolBlend {blendOption=Linear}
    boneToModify {boneName=dyn_hair_bk_01}
    translation {y=-1}
   }
   n89 = "Transform (Modify) Bone - Bone: dyn_hair_bk_01"(ComponentPose=@n88, Alpha=1, Translation="0,-1,0", Rotation="0,0,0") {
    rotationMode=BMM_Additive
    rotationSpace=BCS_BoneSpace
    translationMode=BMM_Additive
    translationSpace=BCS_BoneSpace
    alphaBoolBlend {blendOption=Linear}
    boneToModify {boneName=dyn_hair_bk_01}
    translation {y=-1}
   }
   n90 = "Transform (Modify) Bone - Bone: dyn_hair_r_01"(ComponentPose=@n89, Alpha=1, Translation="0,-2,0", Rotation="0,0,0") {
    rotationMode=BMM_Additive
    rotationSpace=BCS_BoneSpace
    translationMode=BMM_Additive
    translationSpace=BCS_BoneSpace
    alphaBoolBlend {blendOption=Linear}
    boneToModify {boneName=dyn_hair_r_01}
    translation {y=-2}
   }
   n91 = "Transform (Modify) Bone - Bone: dyn_hair_l_01"(ComponentPose=@n90, Alpha=1, Translation="0,-1,0", Rotation="0,0,0") {
    rotationMode=BMM_Additive
    rotationSpace=BCS_BoneSpace
    translationMode=BMM_Additive
    translationSpace=BCS_BoneSpace
    alphaBoolBlend {blendOption=Linear}
    boneToModify {boneName=dyn_hair_l_01}
    translation {y=-1}
   }
   n92 = "Transform (Modify) Bone - Bone: dyn_hair_bk_01"(ComponentPose=@n91, Alpha=1, Translation="0,0,0", Rotation="0,0,0") {
    rotationMode=BMM_Additive
    rotationSpace=BCS_BoneSpace
    translationMode=BMM_Additive
    translationSpace=BCS_BoneSpace
    alphaBoolBlend {blendOption=Linear}
    boneToModify {boneName=dyn_hair_bk_01}
   }
   n93 = "Constraint - dyn_hair_bk_02 (P)"(ComponentPose=@n92, Alpha=1, ConstraintWeights_0=1) {
    alphaBoolBlend {blendOption=Linear}
    boneToModify {boneName=dyn_hair_bk_02}
    constraintSetup [{targetBone {boneName=spine_05}}]
    constraintWeights [1]
   }
   n94 = "Constraint - dyn_hair_l_02 (P)"(ComponentPose=@n93, Alpha=1, ConstraintWeights_0=1) {
    alphaBoolBlend {blendOption=Linear}
    boneToModify {boneName=dyn_hair_l_02}
    constraintSetup [{targetBone {boneName=spine_05}}]
    constraintWeights [1]
   }
   n95 = "Constraint - dyn_hair_r_02 (P)"(ComponentPose=@n94, Alpha=1, ConstraintWeights_0=1) {
    alphaBoolBlend {blendOption=Linear}
    boneToModify {boneName=dyn_hair_r_02}
    constraintSetup [{targetBone {boneName=spine_05}}]
    constraintWeights [1]
   }
   n96 = "Transform (Modify) Bone - Bone: dyn_hair_l_02"!(ComponentPose=@n95, Alpha=1, Translation="0,-1,0", Rotation="0,-10,0") -> (Pose) {
    rotationMode=BMM_Additive
    rotationSpace=BCS_BoneSpace
    translationMode=BMM_Additive
    translationSpace=BCS_BoneSpace
    alphaBoolBlend {blendOption=Linear}
    boneToModify {boneName=dyn_hair_l_02}
    rotation {yaw=-10}
    translation {y=-1}
   }
   n97 = "Transform (Modify) Bone - Bone: dyn_hair_bk_02"!(ComponentPose=@n96, Alpha=1, Translation="0,-1,0", Rotation="0,-10,0") -> (Pose) {
    rotationMode=BMM_Additive
    rotationSpace=BCS_BoneSpace
    translationMode=BMM_Additive
    translationSpace=BCS_BoneSpace
    alphaBoolBlend {blendOption=Linear}
    boneToModify {boneName=dyn_hair_bk_02}
    rotation {yaw=-10}
    translation {y=-1}
   }
   n98 = "Transform (Modify) Bone - Bone: dyn_hair_r_02"!(ComponentPose=@n97, Alpha=1, Translation="0,-1,0", Rotation="0,-10,0") -> (Pose) {
    rotationMode=BMM_Additive
    rotationSpace=BCS_BoneSpace
    translationMode=BMM_Additive
    translationSpace=BCS_BoneSpace
    alphaBoolBlend {blendOption=Linear}
    boneToModify {boneName=dyn_hair_r_02}
    rotation {yaw=-10}
    translation {y=-1}
   }
   n99 = "Transform (Modify) Bone - Bone: dyn_hair_l_02"(ComponentPose=@n98, Alpha=1, Translation="8,-1,0", Rotation="0,40,0") {
    rotationMode=BMM_Additive
    rotationSpace=BCS_BoneSpace
    translationMode=BMM_Additive
    translationSpace=BCS_BoneSpace
    alphaBoolBlend {blendOption=Linear}
    boneToModify {boneName=dyn_hair_l_02}
    rotation {yaw=40}
    translation=<8 -1>
   }
   n100 = "Transform (Modify) Bone - Bone: dyn_hair_bk_02"(ComponentPose=@n99, Alpha=1, Translation="8,-1,0", Rotation="0,40,0") {
    rotationMode=BMM_Additive
    rotationSpace=BCS_BoneSpace
    translationMode=BMM_Additive
    translationSpace=BCS_BoneSpace
    alphaBoolBlend {blendOption=Linear}
    boneToModify {boneName=dyn_hair_bk_02}
    rotation {yaw=40}
    translation=<8 -1>
   }
   n101 = "Transform (Modify) Bone - Bone: dyn_hair_r_02"(ComponentPose=@n100, Alpha=1, Translation="8,-1,0", Rotation="0,40,0") {
    rotationMode=BMM_Additive
    rotationSpace=BCS_BoneSpace
    translationMode=BMM_Additive
    translationSpace=BCS_BoneSpace
    alphaBoolBlend {blendOption=Linear}
    boneToModify {boneName=dyn_hair_r_02}
    rotation {yaw=40}
    translation=<8 -1>
   }
   n102 = "Transform (Modify) Bone - Bone: dyn_hair_l_02"(ComponentPose=@n101, Alpha=1, Translation="0,-2,-4", Rotation="0,0,0") {
    rotationMode=BMM_Additive
    rotationSpace=BCS_BoneSpace
    translationMode=BMM_Additive
    translationSpace=BCS_BoneSpace
    alphaBoolBlend {blendOption=Linear}
    boneToModify {boneName=dyn_hair_l_02}
    translation {y=-2, z=-4}
   }
   n103 = "Transform (Modify) Bone - Bone: dyn_hair_bk_02"(ComponentPose=@n102, Alpha=1, Translation="0,0,-3", Rotation="0,0,0") {
    rotationMode=BMM_Additive
    rotationSpace=BCS_BoneSpace
    translationMode=BMM_Additive
    translationSpace=BCS_BoneSpace
    alphaBoolBlend {blendOption=Linear}
    boneToModify {boneName=dyn_hair_bk_02}
    translation {z=-3}
   }
   n104 = "Transform (Modify) Bone - Bone: dyn_hair_l_02"(ComponentPose=@n103, Alpha=1, Translation="0,-3,0", Rotation="0,0,0") {
    rotationMode=BMM_Additive
    rotationSpace=BCS_BoneSpace
    translationMode=BMM_Additive
    translationSpace=BCS_BoneSpace
    alphaBoolBlend {blendOption=Linear}
    boneToModify {boneName=dyn_hair_l_02}
    translation {y=-3}
   }
   n105 = "Transform (Modify) Bone - Bone: dyn_hair_bk_02"(ComponentPose=@n104, Alpha=1, Translation="0,-1,0", Rotation="0,0,0") {
    rotationMode=BMM_Additive
    rotationSpace=BCS_BoneSpace
    translationMode=BMM_Additive
    translationSpace=BCS_BoneSpace
    alphaBoolBlend {blendOption=Linear}
    boneToModify {boneName=dyn_hair_bk_02}
    translation {y=-1}
   }
   n106 = "Transform (Modify) Bone - Bone: dyn_hair_bk_02"(ComponentPose=@n105, Alpha=1, Translation="0,-1,0", Rotation="0,0,0") {
    rotationMode=BMM_Additive
    rotationSpace=BCS_BoneSpace
    translationMode=BMM_Additive
    translationSpace=BCS_BoneSpace
    alphaBoolBlend {blendOption=Linear}
    boneToModify {boneName=dyn_hair_bk_02}
    translation {y=-1}
   }
   n107 = "Transform (Modify) Bone - Bone: dyn_hair_r_02"(ComponentPose=@n106, Alpha=1, Translation="0,-5,0", Rotation="0,0,0") {
    rotationMode=BMM_Additive
    rotationSpace=BCS_BoneSpace
    translationMode=BMM_Additive
    translationSpace=BCS_BoneSpace
    alphaBoolBlend {blendOption=Linear}
    boneToModify {boneName=dyn_hair_r_02}
    translation {y=-5}
   }
   n108 = "Transform (Modify) Bone - Bone: dyn_hair_l_02"(ComponentPose=@n107, Alpha=1, Translation="0,-3.5,0", Rotation="0,0,0") {
    rotationMode=BMM_Additive
    rotationSpace=BCS_BoneSpace
    translationMode=BMM_Additive
    translationSpace=BCS_BoneSpace
    alphaBoolBlend {blendOption=Linear}
    boneToModify {boneName=dyn_hair_l_02}
    translation {y=-3.5}
   }
   n109 = "Transform (Modify) Bone - Bone: dyn_hair_bk_02"(ComponentPose=@n108, Alpha=1, Translation="0,-2,0", Rotation="0,0,0") {
    rotationMode=BMM_Additive
    rotationSpace=BCS_BoneSpace
    translationMode=BMM_Additive
    translationSpace=BCS_BoneSpace
    alphaBoolBlend {blendOption=Linear}
    boneToModify {boneName=dyn_hair_bk_02}
    translation {y=-2}
   }
   n110 = "Constraint - dyn_hair_bk_03 (P)"(ComponentPose=@n109, Alpha=1, ConstraintWeights_0=1) {
    alphaBoolBlend {blendOption=Linear}
    boneToModify {boneName=dyn_hair_bk_03}
    constraintSetup [{targetBone {boneName=spine_04}}]
    constraintWeights [1]
   }
   n111 = "Constraint - dyn_hair_l_03 (P)"(ComponentPose=@n110, Alpha=1, ConstraintWeights_0=1) {
    alphaBoolBlend {blendOption=Linear}
    boneToModify {boneName=dyn_hair_l_03}
    constraintSetup [{targetBone {boneName=spine_04}}]
    constraintWeights [1]
   }
   n112 = "Constraint - dyn_hair_r_03 (P)"(ComponentPose=@n111, Alpha=1, ConstraintWeights_0=1) {
    alphaBoolBlend {blendOption=Linear}
    boneToModify {boneName=dyn_hair_r_03}
    constraintSetup [{targetBone {boneName=spine_04}}]
    constraintWeights [1]
   }
   n113 = "Transform (Modify) Bone - Bone: dyn_hair_l_03"!(ComponentPose=@n112, Alpha=1, Translation="0,1,0", Rotation="0,10,0") -> (Pose) {
    rotationMode=BMM_Additive
    rotationSpace=BCS_BoneSpace
    translationMode=BMM_Additive
    translationSpace=BCS_BoneSpace
    alphaBoolBlend {blendOption=Linear}
    boneToModify {boneName=dyn_hair_l_03}
    rotation {yaw=10}
    translation {y=1}
   }
   n114 = "Transform (Modify) Bone - Bone: dyn_hair_bk_03"!(ComponentPose=@n113, Alpha=1, Translation="0,1,0", Rotation="0,10,0") -> (Pose) {
    rotationMode=BMM_Additive
    rotationSpace=BCS_BoneSpace
    translationMode=BMM_Additive
    translationSpace=BCS_BoneSpace
    alphaBoolBlend {blendOption=Linear}
    boneToModify {boneName=dyn_hair_bk_03}
    rotation {yaw=10}
    translation {y=1}
   }
   n115 = "Transform (Modify) Bone - Bone: dyn_hair_r_03"!(ComponentPose=@n114, Alpha=1, Translation="0,1,0", Rotation="0,10,0") -> (Pose) {
    rotationMode=BMM_Additive
    rotationSpace=BCS_BoneSpace
    translationMode=BMM_Additive
    translationSpace=BCS_BoneSpace
    alphaBoolBlend {blendOption=Linear}
    boneToModify {boneName=dyn_hair_r_03}
    rotation {yaw=10}
    translation {y=1}
   }
   n116 = "Transform (Modify) Bone - Bone: dyn_hair_l_03"(ComponentPose=@n115, Alpha=1, Translation="8,-2,1", Rotation="0,20,0") {
    rotationMode=BMM_Additive
    rotationSpace=BCS_BoneSpace
    translationMode=BMM_Additive
    translationSpace=BCS_BoneSpace
    alphaBoolBlend {blendOption=Linear}
    boneToModify {boneName=dyn_hair_l_03}
    rotation {yaw=20}
    translation=<8 -2 1>
   }
   n117 = "Transform (Modify) Bone - Bone: dyn_hair_bk_03"(ComponentPose=@n116, Alpha=1, Translation="8,-2,0", Rotation="0,20,0") {
    rotationMode=BMM_Additive
    rotationSpace=BCS_BoneSpace
    translationMode=BMM_Additive
    translationSpace=BCS_BoneSpace
    alphaBoolBlend {blendOption=Linear}
    boneToModify {boneName=dyn_hair_bk_03}
    rotation {yaw=20}
    translation=<8 -2>
   }
   n118 = "Transform (Modify) Bone - Bone: dyn_hair_r_03"(ComponentPose=@n117, Alpha=1, Translation="8,-2,-1", Rotation="0,20,0") {
    rotationMode=BMM_Additive
    rotationSpace=BCS_BoneSpace
    translationMode=BMM_Additive
    translationSpace=BCS_BoneSpace
    alphaBoolBlend {blendOption=Linear}
    boneToModify {boneName=dyn_hair_r_03}
    rotation {yaw=20}
    translation=<8 -2 -1>
   }
   n119 = "Transform (Modify) Bone - Bone: dyn_hair_l_03"(ComponentPose=@n118, Alpha=1, Translation="0,-2.5,0", Rotation="0,0,0") {
    rotationMode=BMM_Additive
    rotationSpace=BCS_BoneSpace
    translationMode=BMM_Additive
    translationSpace=BCS_BoneSpace
    alphaBoolBlend {blendOption=Linear}
    boneToModify {boneName=dyn_hair_l_03}
    translation {y=-2.5}
   }
   n120 = "Transform (Modify) Bone - Bone: dyn_hair_bk_03"(ComponentPose=@n119, Alpha=1, Translation="0,-1,0", Rotation="0,0,0") {
    rotationMode=BMM_Additive
    rotationSpace=BCS_BoneSpace
    translationMode=BMM_Additive
    translationSpace=BCS_BoneSpace
    alphaBoolBlend {blendOption=Linear}
    boneToModify {boneName=dyn_hair_bk_03}
    translation {y=-1}
   }
   n121 = "Transform (Modify) Bone - Bone: dyn_hair_bk_03"(ComponentPose=@n120, Alpha=1, Translation="0,-1,0", Rotation="0,0,0") {
    rotationMode=BMM_Additive
    rotationSpace=BCS_BoneSpace
    translationMode=BMM_Additive
    translationSpace=BCS_BoneSpace
    alphaBoolBlend {blendOption=Linear}
    boneToModify {boneName=dyn_hair_bk_03}
    translation {y=-1}
   }
   n122 = "Transform (Modify) Bone - Bone: dyn_hair_r_03"(ComponentPose=@n121, Alpha=1, Translation="0,-4,0", Rotation="0,0,0") {
    rotationMode=BMM_Additive
    rotationSpace=BCS_BoneSpace
    translationMode=BMM_Additive
    translationSpace=BCS_BoneSpace
    alphaBoolBlend {blendOption=Linear}
    boneToModify {boneName=dyn_hair_r_03}
    translation {y=-4}
   }
   Reroute Node(InputPin:Component Space Pose Link Structure=@n122) -> (OutputPin:Component Space Pose Link Structure)
   Get Override World Gravity()
   Get Component Linear Acc Scale()
  ]}
 EventGraph {nodes [
   Event BlueprintUpdateAnimation()
   > n1 = Sequence!(execute) -> (then_0, then_1, then_2:Exec)
   > n2 = Gravity Override State Function DEPRECATED(CharacterPartAnimInstance=Self-Reference(), GameTimeInSeconds=GetGameTimeInSeconds(), GravityOverrideParams=Get Gravity Override Params())
   > Set Override World Gravity(Override World Gravity:Vector=@n2) -> (Output_Get:Vector)
   n3 = CheckForBackbling()
   Set Wearing_Backpack(execute=@n1.then_1, Wearing_Backpack:bool=AND Boolean(A=@n3, B=NOT Boolean(A=@n3.bIsBackblingACape))) -> (Output_Get:bool)
   Event OnCharacterCustomizationLoaded()
  ]}
]
variables [
 Clavicle_L_Lift_Up {type=number}
 Clavicle_R_Lift_Up {type=number}
 Clavicle_R_Turn_Back {type=number}
 Component Linear Acc Scale {type=Vector, default=<0.5 0.5 0.5>}
 Component Linear Vel Scale {type=Vector, default=<0.5 0.5 0.5>}
 Delta Time X {type=number}
 Gravity Override Params {type=GravityOverrideParamsStruct, default {tactical_Sprint {gravityOverride=<0 -50 -500>}}}
 Head_Rot_Bk {type=number}
 Head_Rot_R {type=number}
 InSkyWithoutParachute {type=boolean}
 Override World Gravity {type=Vector, default=<0 0 -980>}
 SimpleJump {type=boolean}
 SkelMesh {type=SkeletalMeshComponent}
 Wearing_Backpack {type=boolean}
]
